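# Control Rig for the ParadeLotus medium male body cosmetic that adds procedural corrective deformation to body and skirt bones.
# The forward solve runs shared FK control and DeformRig functions, then local lat, skirt, biceps and shoulder fixes.
# Spherical pose readers on thighs and arms weight bone offsets, while remapped clavicle rotation drives the lats.
# The skirt also blends poses for crouch (scaled by pelvis height), slide, DBNO and tactical states.
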ControlRigBlueprint /BRCosmetics/Characters/Player/Male/Medium/Bodies/M_MED_ParadeLotus/Meshes/M_MED_ParadeLotus_CtrlRig2
graphs [
 New Function_ContainedGraph {nodes [
   n1 = RigVMFunctionEntryNode()
   RigVMFunction_Sequence = &RigVMFunction_Sequence(ExecuteContext=@n1.ExecuteContext)
   &RigUnit_ParentConstraint(
    AdvancedSettings=(InterpolationType=Average,RotationOrderForFilter=XZY)
    Child=(Type=Bone,Name="clavicle_l_set_rot")
    Parents=((Item=(Type=Bone,Name="clavicle_l"),Weight=1))
    Weight=1
    ExecutePin=@RigVMFunction_Sequence.A
   )
   RigUnit_GetTransform = &RigUnit_GetTransform(Item=(Type=Bone,Name="clavicle_l_set_rot"), Space=LocalSpace, bInitial=False)
   RigVMFunction_MathQuaternionToEuler = &RigVMFunction_MathQuaternionToEuler(RotationOrder=ZYX, Value=@RigUnit_GetTransform.Transform.Rotation)
   Remap = &RigVMFunction_MathDoubleRemap(SourceMaximum=-50, SourceMinimum=0, TargetMaximum=1, TargetMinimum=0, Value=@RigVMFunction_MathQuaternionToEuler.Result.Y)
   > &RigUnit_ModifyTransforms(
    ItemToModify=((Item=(Type=Bone,Name="deform_lat_l"),Transform=(Rotation=(X=0,Y=0,Z=0,W=1),Translation=(X=0,Y=0,Z=6),Scale3D=(X=1,Y=1,Z=0.5))))
    WeightMinimum=0
    Weight=@Remap.Result
   )
   > &RigUnit_ParentConstraint(
    AdvancedSettings=(InterpolationType=Average,RotationOrderForFilter=XZY)
    Child=(Type=Bone,Name="clavicle_r_set_rot")
    Parents=((Item=(Type=Bone,Name="clavicle_r"),Weight=1))
    Weight=1
   )
   n2 = &RigUnit_GetTransform(Item=(Type=Bone,Name="clavicle_r_set_rot"), Space=LocalSpace, bInitial=False)
   > &RigUnit_ModifyTransforms(
    ItemToModify=((Item=(Type=Bone,Name="deform_lat_r"),Transform=(Rotation=(X=0,Y=0,Z=0,W=1),Translation=(X=0,Y=0,Z=-6),Scale3D=(X=1,Y=1,Z=0.5))))
    WeightMinimum=0
    Weight=&RigVMFunction_MathDoubleRemap(
     SourceMaximum=-50
     SourceMinimum=0
     TargetMaximum=1
     TargetMinimum=0
     Value=&RigVMFunction_MathQuaternionToEuler(RotationOrder=ZYX, Value=@n2.Transform.Rotation).Result.Y
    ).Result
   )
   RigVMFunctionReturnNode(ExecuteContext=@RigVMFunction_Sequence.B)
  ]}
 New Function_ContainedGraph {nodes [
   RigVMFunctionEntryNode()
   > RigVMFunction_Sequence = RigVMAggregateNode()
   n1 = &RigUnit_SphericalPoseReader(
    ActiveRegionScaleFactors=(PositiveWidth=1,NegativeWidth=1,PositiveHeight=1,NegativeHeight=1)
    ActiveRegionSize=0.05
    Debug=(bDrawDebug=false,bDraw2D=False,bDrawLocalAxes=False,DebugScale=25,DebugSegments=20,DebugThickness=0.25)
    DriverAxis=(X=-1,Y=0,Z=0)
    DriverItem=(Type=Bone,Name="thigh_l")
    FalloffRegionScaleFactors=(PositiveWidth=1,NegativeWidth=1,PositiveHeight=0.9,NegativeHeight=0.9)
    FalloffSize=0.63
    FlipHeightScaling=False
    FlipWidthScaling=False
    OptionalParentItem=(Type=Bone,Name="None")
    RotationOffset=(X=90,Y=0,Z=-30)
    ExecutePin=@RigVMFunction_Sequence.A
   )
   > &RigUnit_ModifyTransforms(
    ItemToModify="((Item=(Type=Bone,Name=\"base_skirt_front_l_0\"),Transform=(Rotation=(X=-0.0201,Y=-0.5754,Z=-0.2787,W=0.7687),Translation=(X=-7.0003,Y=5.0003,Z=-4.0002),Scale3D=(X=1,Y=1,Z=1))),(Item=(Type=Bone,Name=\"base_skirt_side_l_0\"),Transform=(Rotation=(X=-0.0264,Y=-0. ...(truncated)"
    WeightMinimum=0
    Weight=@n1.OutputParam
   )
   > n2 = &RigUnit_SphericalPoseReader(
    ActiveRegionScaleFactors=(PositiveWidth=1,NegativeWidth=1,PositiveHeight=1,NegativeHeight=1)
    ActiveRegionSize=0.05
    Debug=(bDrawDebug=false,bDraw2D=False,bDrawLocalAxes=False,DebugScale=25,DebugSegments=20,DebugThickness=0.25)
    DriverAxis=(X=1,Y=0,Z=0)
    DriverItem=(Type=Bone,Name="thigh_r")
    FalloffRegionScaleFactors=(PositiveWidth=1,NegativeWidth=1,PositiveHeight=1,NegativeHeight=1)
    FalloffSize=0.63
    FlipHeightScaling=true
    FlipWidthScaling=true
    OptionalParentItem=(Type=Bone,Name="None")
    RotationOffset=(X=-90,Y=0,Z=-30)
   )
   > &RigUnit_ModifyTransforms(
    ItemToModify="((Item=(Type=Bone,Name=\"base_skirt_front_r_0\"),Transform=(Rotation=(X=-0.0912,Y=0.5016,Z=-0.4354,W=0.742),Translation=(X=-10.9998,Y=3.1998,Z=3),Scale3D=(X=1,Y=1,Z=1))),(Item=(Type=Bone,Name=\"base_skirt_side_r_0\"),Transform=(Rotation=(X=-0.0341,Y=0.3646,Z=- ...(truncated)"
    WeightMinimum=0
    Weight=@n2.OutputParam
   )
   Maximum = &RigVMFunction_MathDoubleMax(B=@n1.OutputParam, A=@n2.OutputParam)
   > &RigUnit_ModifyTransforms(
    ItemToModify=((Item=(Type=Bone,Name="base_skirt_front_0"),Transform=(Rotation=(X=0,Y=0,Z=-0.9465,W=0.3227),Translation=(X=0,Y=0,Z=0),Scale3D=(X=1,Y=1,Z=1))))
    WeightMinimum=0
    Weight=@Maximum.Result
   )
   > n3 = &RigUnit_SphericalPoseReader(
    ActiveRegionScaleFactors=(PositiveWidth=1,NegativeWidth=1,PositiveHeight=1,NegativeHeight=1)
    ActiveRegionSize=0.05
    Debug=(bDrawDebug=false,bDraw2D=False,bDrawLocalAxes=False,DebugScale=25,DebugSegments=20,DebugThickness=0.25)
    DriverAxis=(X=-1,Y=0,Z=0)
    DriverItem=(Type=Bone,Name="thigh_l")
    FalloffRegionScaleFactors=(PositiveWidth=1.1,NegativeWidth=1.1,PositiveHeight=1.1,NegativeHeight=1.1)
    FalloffSize=0.67
    FlipHeightScaling=False
    FlipWidthScaling=False
    OptionalParentItem=(Type=Bone,Name="None")
    RotationOffset=(X=180,Y=30,Z=0)
   )
   > &RigUnit_ModifyTransforms(
    ItemToModify="((Item=(Type=Bone,Name=\"base_skirt_side_l_0\"),Transform=(Rotation=(X=-0.1739,Y=-0.2103,Z=0.526,W=-0.8055),Translation=(X=-4.0001,Y=2,Z=0),Scale3D=(X=1,Y=1,Z=1))),(Item=(Type=Bone,Name=\"base_skirt_front_l_0\"),Transform=(Rotation=(X=0.028,Y=0.688,Z=0.0589,W= ...(truncated)"
    WeightMinimum=0
    Weight=@n3.OutputParam
   )
   > n4 = &RigUnit_SphericalPoseReader(
    ActiveRegionScaleFactors=(PositiveWidth=1,NegativeWidth=1,PositiveHeight=1,NegativeHeight=1)
    ActiveRegionSize=0.05
    Debug=(bDrawDebug=false,bDraw2D=False,bDrawLocalAxes=False,DebugScale=25,DebugSegments=20,DebugThickness=0.25)
    DriverAxis=(X=-1,Y=0,Z=0)
    DriverItem=(Type=Bone,Name="thigh_l")
    FalloffRegionScaleFactors=(PositiveWidth=1,NegativeWidth=1,PositiveHeight=1,NegativeHeight=1)
    FalloffSize=0.63
    FlipHeightScaling=False
    FlipWidthScaling=False
    OptionalParentItem=(Type=Bone,Name="None")
    RotationOffset=(X=140,Y=30,Z=0)
   )
   > &RigUnit_OffsetTransformForItem(
    Item=(Type=Bone,Name="base_skirt_front_l_0")
    OffsetTransform=(Rotation=(X=0,Y=0,Z=-0.7071,W=0.7071),Translation=(X=0,Y=0,Z=0),Scale3D=(X=1,Y=1,Z=1))
    bPropagateToChildren=True
    Weight=@n4.OutputParam
   )
   > n5 = &RigUnit_SphericalPoseReader(
    ActiveRegionScaleFactors=(PositiveWidth=1,NegativeWidth=1,PositiveHeight=1,NegativeHeight=1)
    ActiveRegionSize=0.05
    Debug=(bDrawDebug=false,bDraw2D=False,bDrawLocalAxes=False,DebugScale=25,DebugSegments=20,DebugThickness=0.25)
    DriverAxis=(X=1,Y=0,Z=0)
    DriverItem=(Type=Bone,Name="thigh_r")
    FalloffRegionScaleFactors=(PositiveWidth=1,NegativeWidth=1,PositiveHeight=1,NegativeHeight=1)
    FalloffSize=0.67
    FlipHeightScaling=False
    FlipWidthScaling=False
    OptionalParentItem=(Type=Bone,Name="None")
    RotationOffset=(X=0,Y=30,Z=0)
   )
   > &RigUnit_ModifyTransforms(
    ItemToModify="((Item=(Type=Bone,Name=\"base_skirt_side_r_0\"),Transform=(Rotation=(X=0,Y=0,Z=-0.8614,W=0.508),Translation=(X=0,Y=0,Z=0),Scale3D=(X=1,Y=1,Z=1))),(Item=(Type=Bone,Name=\"base_skirt_front_r_0\"),Transform=(Rotation=(X=0.0468,Y=-0.5396,Z=-0.1847,W=0.82),Translat ...(truncated)"
    WeightMinimum=0
    Weight=@n5.OutputParam
   )
   > n6 = &RigUnit_SphericalPoseReader(
    ActiveRegionScaleFactors=(PositiveWidth=1,NegativeWidth=1,PositiveHeight=1,NegativeHeight=1)
    ActiveRegionSize=0.05
    Debug=(bDrawDebug=false,bDraw2D=False,bDrawLocalAxes=False,DebugScale=25,DebugSegments=20,DebugThickness=0.25)
    DriverAxis=(X=1,Y=0,Z=0)
    DriverItem=(Type=Bone,Name="thigh_r")
    FalloffRegionScaleFactors=(PositiveWidth=1,NegativeWidth=1,PositiveHeight=0.9,NegativeHeight=0.9)
    FalloffSize=0.63
    FlipHeightScaling=False
    FlipWidthScaling=False
    OptionalParentItem=(Type=Bone,Name="None")
    RotationOffset=(X=-40,Y=30,Z=0)
   )
   > &RigUnit_OffsetTransformForItem(
    Item=(Type=Bone,Name="base_skirt_front_r_0")
    OffsetTransform=(Rotation=(X=0,Y=0,Z=-0.4329,W=0.9014),Translation=(X=0,Y=0,Z=0),Scale3D=(X=1,Y=1,Z=1))
    bPropagateToChildren=True
    Weight=@n6.OutputParam
   )
   > n7 = &RigUnit_SphericalPoseReader(
    ActiveRegionScaleFactors=(PositiveWidth=1,NegativeWidth=1,PositiveHeight=1,NegativeHeight=1)
    ActiveRegionSize=0.05
    Debug=(bDrawDebug=false,bDraw2D=False,bDrawLocalAxes=False,DebugScale=25,DebugSegments=20,DebugThickness=0.25)
    DriverAxis=(X=-1,Y=0,Z=0)
    DriverItem=(Type=Bone,Name="thigh_l")
    FalloffRegionScaleFactors=(PositiveWidth=0.75,NegativeWidth=0.75,PositiveHeight=0.9,NegativeHeight=0.9)
    FalloffSize=0.63
    FlipHeightScaling=False
    FlipWidthScaling=False
    OptionalParentItem=(Type=Bone,Name="None")
    RotationOffset=(X=-90,Y=0,Z=10)
   )
   > &RigUnit_ModifyTransforms(
    ItemToModify=((Item=(Type=Bone,Name="base_skirt_back_side_l_0"),Transform=(Rotation=(X=-0.0739,Y=0.183,Z=-0.683,W=0.7032),Translation=(X=0,Y=0,Z=0),Scale3D=(X=1,Y=1,Z=1))))
    WeightMinimum=0
    Weight=@n7.OutputParam
   )
   > n8 = &RigUnit_SphericalPoseReader(
    ActiveRegionScaleFactors=(PositiveWidth=1,NegativeWidth=1,PositiveHeight=1,NegativeHeight=1)
    ActiveRegionSize=0.05
    Debug=(bDrawDebug=false,bDraw2D=False,bDrawLocalAxes=False,DebugScale=25,DebugSegments=20,DebugThickness=0.25)
    DriverAxis=(X=1,Y=0,Z=0)
    DriverItem=(Type=Bone,Name="thigh_r")
    FalloffRegionScaleFactors=(PositiveWidth=0.75,NegativeWidth=0.75,PositiveHeight=0.9,NegativeHeight=0.9)
    FalloffSize=0.63
    FlipHeightScaling=False
    FlipWidthScaling=False
    OptionalParentItem=(Type=Bone,Name="None")
    RotationOffset=(X=90,Y=0,Z=10)
   )
   > &RigUnit_ModifyTransforms(
    ItemToModify=((Item=(Type=Bone,Name="base_skirt_back_side_r_0"),Transform=(Rotation=(X=0.1048,Y=-0.2754,Z=-0.6831,W=0.6683),Translation=(X=0,Y=0,Z=0),Scale3D=(X=1,Y=1,Z=1))))
    WeightMinimum=0
    Weight=@n8.OutputParam
   )
   > &RigUnit_ModifyTransforms(
    ItemToModify=((Item=(Type=Bone,Name="base_skirt_back_0"),Transform=(Rotation=(X=0,Y=0,Z=-0.766,W=0.6428),Translation=(X=0,Y=0,Z=0),Scale3D=(X=1,Y=1,Z=1))))
    WeightMinimum=0
    Weight=&RigVMFunction_MathDoubleMax(A=@n7.OutputParam, B=@n8.OutputParam).Result
   )
   > n9 = &RigUnit_SphericalPoseReader(
    ActiveRegionScaleFactors=(PositiveWidth=1,NegativeWidth=1,PositiveHeight=1,NegativeHeight=1)
    ActiveRegionSize=0.05
    Debug=(bDrawDebug=false,bDraw2D=False,bDrawLocalAxes=False,DebugScale=25,DebugSegments=20,DebugThickness=0.25)
    DriverAxis=(X=-1,Y=0,Z=0)
    DriverItem=(Type=Bone,Name="thigh_l")
    FalloffRegionScaleFactors=(PositiveWidth=1,NegativeWidth=1,PositiveHeight=0.9,NegativeHeight=0.9)
    FalloffSize=0.63
    FlipHeightScaling=False
    FlipWidthScaling=false
    OptionalParentItem=(Type=Bone,Name="None")
    RotationOffset=(X=-135,Y=0,Z=0)
   )
   > &RigUnit_OffsetTransformForItem(
    Item=(Type=Bone,Name="base_skirt_back_side_l_0")
    OffsetTransform=(Rotation=(X=0.0222,Y=-0.1522,Z=-0.1311,W=0.9794),Translation=(X=0,Y=0,Z=0),Scale3D=(X=1,Y=1,Z=1))
    bPropagateToChildren=True
    Weight=@n9.OutputParam
   )
   > n10 = &RigUnit_SphericalPoseReader(
    ActiveRegionScaleFactors=(PositiveWidth=1,NegativeWidth=1,PositiveHeight=1,NegativeHeight=1)
    ActiveRegionSize=0.05
    Debug=(bDrawDebug=false,bDraw2D=False,bDrawLocalAxes=False,DebugScale=25,DebugSegments=20,DebugThickness=0.25)
    DriverAxis=(X=1,Y=0,Z=0)
    DriverItem=(Type=Bone,Name="thigh_r")
    FalloffRegionScaleFactors=(PositiveWidth=1,NegativeWidth=1,PositiveHeight=0.9,NegativeHeight=0.9)
    FalloffSize=0.63
    FlipHeightScaling=False
    FlipWidthScaling=False
    OptionalParentItem=(Type=Bone,Name="None")
    RotationOffset=(X=45,Y=0,Z=0)
   )
   > &RigUnit_OffsetTransformForItem(
    Item=(Type=Bone,Name="base_skirt_back_side_r_0")
    OffsetTransform=(Rotation=(X=0.0807,Y=0.2723,Z=-0.2723,W=0.9193),Translation=(X=0,Y=0,Z=0),Scale3D=(X=1,Y=1,Z=1))
    bPropagateToChildren=True
    Weight=@n10.OutputParam
   )
   RigVMFunctionReturnNode(ExecuteContext=@RigVMFunction_Sequence.G)
   n11 = &RigVMFunction_AlphaInterp(
    InterpSpeedDecreasing=10
    InterpSpeedIncreasing=10
    Scale=1
    bClampResult=False
    Value=&RigVMFunction_MathBoolToFloat(Value=RigVMVariableNode() {variable="Istactical "}.Value).Result
   )
   &RigUnit_ModifyTransforms(
    ItemToModify="((Item=(Type=Bone,Name=\"dyn_skirt_back_side_l_1\"),Transform=(Rotation=(X=-0.0653,Y=0.0969,Z=-0.1314,W=0.9844),Translation=(X=0,Y=0,Z=0),Scale3D=(X=1,Y=1,Z=1))),(Item=(Type=Bone,Name=\"dyn_skirt_back_side_l_2\"),Transform=(Rotation=(X=-0.03,Y=-0.1043,Z=-0.274 ...(truncated)"
    WeightMinimum=0
    Weight=@n11.Result
    ExecutePin=@RigVMFunction_Sequence.F
   )
   n12 = &RigVMFunction_AnimEvalRichCurve(
    Curve=(EditorCurveData=(Keys=((InterpMode=RCIM_Cubic),(InterpMode=RCIM_Cubic,Time=0.5,Value=1),(InterpMode=RCIM_Cubic,Time=0.8))))
    SourceMaximum=2
    SourceMinimum=0
    TargetMinimum=0
    Value=&RigVMFunction_MathDoubleRemap(
     SourceMaximum=75
     SourceMinimum=45
     TargetMaximum=1
     TargetMinimum=0
     bClamp=false
     Value=&RigUnit_GetRelativeTransformForItem(Child=(Type=Bone,Name="pelvis"), Parent=(Type=Bone,Name="root"), bChildInitial=False, bParentInitial=False).RelativeTransform.Translation.Z
    ).Result
   )
   RigVMVariableNode(Value=@n12.Result, ExecuteContext=@RigVMFunction_Sequence.D) {variable=CrouchWalk}
   n13 = &RigVMFunction_AlphaInterp(
    InterpSpeedDecreasing=10
    InterpSpeedIncreasing=10
    Scale=1
    bClampResult=False
    Value=&RigVMFunction_MathBoolToFloat(Value=RigVMVariableNode() {variable=IsSliding}.Value).Result
   )
   > &RigUnit_ModifyTransforms(
    ItemToModify="((Item=(Type=Bone,Name=\"base_skirt_front_0\"),Transform=(Rotation=(X=0,Y=0,Z=-0.1655,W=0.9862),Translation=(X=0,Y=0,Z=0),Scale3D=(X=1,Y=1,Z=1))),(Item=(Type=Bone,Name=\"base_skirt_front_l_0\"),Transform=(Rotation=(X=0,Y=0,Z=-0.342,W=0.9397),Translation=(X=0,Y ...(truncated)"
    WeightMinimum=0
    Weight=@n13.Result
   )
   n14 = &RigVMFunction_AlphaInterp(
    InterpSpeedDecreasing=10
    InterpSpeedIncreasing=10
    Scale=1
    bClampResult=False
    Value=&RigVMFunction_MathBoolToFloat(Value=RigVMVariableNode() {variable=isDBNO}.Value).Result
   )
   &RigUnit_ModifyTransforms(
    ItemToModify="((Item=(Type=Bone,Name=\"dyn_skirt_back_side_l_0\"),Transform=(Rotation=(X=0,Y=0,Z=-0.1305,W=0.9914),Translation=(X=0,Y=0,Z=0),Scale3D=(X=1,Y=1,Z=1))),(Item=(Type=Bone,Name=\"dyn_skirt_back_side_l_1\"),Transform=(Rotation=(X=0,Y=0,Z=-0.0872,W=0.9962),Translati ...(truncated)"
    WeightMinimum=0
    Weight=@n14.Result
    ExecutePin=@RigVMFunction_Sequence.E
   )
   Remap = &RigVMFunction_MathDoubleRemap(
    SourceMaximum=45
    SourceMinimum=89
    TargetMaximum=1
    TargetMinimum=0
    Value=&RigUnit_GetRelativeTransformForItem(Child=(Type=Bone,Name="pelvis"), Parent=(Type=Bone,Name="root"), bChildInitial=False, bParentInitial=False).RelativeTransform.Translation.Z
   )
   RigVMFunction_MathBoolToFloat = &RigVMFunction_MathBoolToFloat(Value=RigVMVariableNode() {variable=IsCrouching}.Value)
   Multiply = &RigVMFunction_MathDoubleMul(A=@Remap.Result, B=@RigVMFunction_MathBoolToFloat.Result)
   &RigUnit_ModifyTransforms(
    ItemToModify="((Item=(Type=Bone,Name=\"base_skirt_side_l_0\"),Transform=(Rotation=(X=0.0777,Y=-0.0352,Z=0.0253,W=0.996),Translation=(X=0.4028,Y=-3.1625,Z=-1.0036),Scale3D=(X=1,Y=1,Z=1))),(Item=(Type=Bone,Name=\"base_skirt_back_side_l_0\"),Transform=(Rotation=(X=0.2436,Y=0.1 ...(truncated)"
    WeightMinimum=0
    Weight=@Multiply.Result
    ExecutePin=@RigVMFunction_Sequence.C
   )
   > RigVMVariableNode(Value=@Multiply.Result) {variable=GetCrouchValue}
   &RigVMFunction_MathDoubleMul(B=@n13.Result, A=@n12.Result)
  ]}
 New Function_ContainedGraph {nodes [
   n1 = RigVMFunctionEntryNode()
   RigVMFunction_Sequence = &RigVMFunction_Sequence(ExecuteContext=@n1.ExecuteContext)
   n2 = &RigUnit_SphericalPoseReader(
    ActiveRegionScaleFactors=(PositiveWidth=2,NegativeWidth=1,PositiveHeight=2,NegativeHeight=2)
    ActiveRegionSize=0.15
    Debug=(bDrawDebug=true,bDraw2D=False,bDrawLocalAxes=False,DebugScale=15,DebugSegments=20,DebugThickness=0.1)
    DriverAxis=(X=-1,Y=0,Z=0)
    DriverItem=(Type=Bone,Name="lowerarm_r")
    FalloffRegionScaleFactors=(PositiveWidth=1,NegativeWidth=1,PositiveHeight=1,NegativeHeight=1)
    FalloffSize=0.5
    FlipHeightScaling=False
    FlipWidthScaling=False
    OptionalParentItem=(Type=Bone,Name="upperarm_twist_02_r")
    RotationOffset=(X=-90,Y=0,Z=0)
    ExecutePin=@RigVMFunction_Sequence.A
   )
   > &RigUnit_OffsetTransformForItem(
    Item=(Type=Bone,Name="upperarm_bicep_r")
    OffsetTransform=(Rotation=(X=0,Y=0,Z=0,W=1),Translation=(X=1,Y=4,Z=-1),Scale3D=(X=1,Y=1,Z=1))
    bPropagateToChildren=True
    Weight=@n2.OutputParam
   )
   > n3 = &RigUnit_SphericalPoseReader(
    ActiveRegionScaleFactors=(PositiveWidth=1,NegativeWidth=2,PositiveHeight=2,NegativeHeight=2)
    ActiveRegionSize=0.15
    Debug=(bDrawDebug=true,bDraw2D=False,bDrawLocalAxes=False,DebugScale=15,DebugSegments=20,DebugThickness=0.1)
    DriverAxis=(X=1,Y=0,Z=0)
    DriverItem=(Type=Bone,Name="lowerarm_l")
    FalloffRegionScaleFactors=(PositiveWidth=1,NegativeWidth=1,PositiveHeight=1,NegativeHeight=1)
    FalloffSize=0.5
    FlipHeightScaling=False
    FlipWidthScaling=False
    OptionalParentItem=(Type=Bone,Name="upperarm_twist_02_l")
    RotationOffset=(X=90,Y=0,Z=0)
   )
   > &RigUnit_OffsetTransformForItem(
    Item=(Type=Bone,Name="upperarm_bicep_l")
    OffsetTransform=(Rotation=(X=0,Y=0,Z=0,W=1),Translation=(X=1,Y=4,Z=1),Scale3D=(X=1,Y=1,Z=1))
    bPropagateToChildren=True
    Weight=@n3.OutputParam
   )
   RigVMFunctionReturnNode(ExecuteContext=@RigVMFunction_Sequence.B)
  ]}
 New Function_ContainedGraph {nodes [
   n1 = RigVMFunctionEntryNode()
   RigVMFunction_Sequence = &RigVMFunction_Sequence(ExecuteContext=@n1.ExecuteContext)
   n2 = &RigUnit_SphericalPoseReader(
    ActiveRegionScaleFactors=(PositiveWidth=2,NegativeWidth=2,PositiveHeight=1,NegativeHeight=1)
    ActiveRegionSize=0.05
    Debug=(bDrawDebug=false,bDraw2D=False,bDrawLocalAxes=False,DebugScale=25,DebugSegments=20,DebugThickness=0.25)
    DriverAxis=(X=1,Y=0,Z=0)
    DriverItem=(Type=Bone,Name="upperarm_l")
    FalloffRegionScaleFactors=(PositiveWidth=1,NegativeWidth=1,PositiveHeight=0.75,NegativeHeight=0.75)
    FalloffSize=0.2
    FlipHeightScaling=False
    FlipWidthScaling=False
    OptionalParentItem=(Type=Bone,Name="None")
    RotationOffset=(X=-15,Y=-30,Z=0)
    ExecutePin=@RigVMFunction_Sequence.A
   )
   > &RigUnit_ModifyTransforms(
    ItemToModify=((Item=(Type=Bone,Name="None"),Transform=(Rotation=(X=0,Y=0,Z=0,W=1),Translation=(X=0,Y=0,Z=0),Scale3D=(X=1,Y=1,Z=1))),(Item=(Type=Bone,Name="deform_upperarm_up_l"),Transform=(Rotation=(X=0,Y=0,Z=0,W=1),Translation=(X=6,Y=0,Z=0),Scale3D=(X=1,Y=1,Z=1))))
    WeightMinimum=0
    Weight=@n2.OutputParam
   )
   > n3 = &RigUnit_SphericalPoseReader(
    ActiveRegionScaleFactors=(PositiveWidth=2,NegativeWidth=2,PositiveHeight=1,NegativeHeight=1)
    ActiveRegionSize=0.05
    Debug=(bDrawDebug=false,bDraw2D=False,bDrawLocalAxes=False,DebugScale=25,DebugSegments=20,DebugThickness=0.25)
    DriverAxis=(X=-1,Y=0,Z=0)
    DriverItem=(Type=Bone,Name="upperarm_r")
    FalloffRegionScaleFactors=(PositiveWidth=1,NegativeWidth=1,PositiveHeight=0.75,NegativeHeight=0.75)
    FalloffSize=0.2
    FlipHeightScaling=False
    FlipWidthScaling=False
    OptionalParentItem=(Type=Bone,Name="None")
    RotationOffset=(X=-15,Y=150,Z=0)
   )
   > &RigUnit_ModifyTransforms(
    ItemToModify=((Item=(Type=Bone,Name="None"),Transform=(Rotation=(X=0,Y=0,Z=0,W=1),Translation=(X=0,Y=0,Z=0),Scale3D=(X=1,Y=1,Z=1))),(Item=(Type=Bone,Name="deform_upperarm_up_r"),Transform=(Rotation=(X=0,Y=0,Z=0,W=1),Translation=(X=-6,Y=0,Z=0),Scale3D=(X=1,Y=1,Z=1))))
    WeightMinimum=0
    Weight=@n3.OutputParam
   )
   RigVMFunctionReturnNode(ExecuteContext=@RigVMFunction_Sequence.B)
  ]}
 RigVMFunctionLibrary {nodes [
   biceps = RigVMCollapseNode(ExecuteContext=())
   deform_lat = RigVMCollapseNode(ExecuteContext=())
   postfix_shoulder = RigVMCollapseNode(ExecuteContext=())
   skirt = RigVMCollapseNode(ExecuteContext=())
  ]}
 RigVMFunction_Sequence_1_ContainedGraph {nodes [
   n1 = RigVMFunctionEntryNode()
   RigVMFunction_Sequence = &RigVMFunction_Sequence(ExecuteContext=@n1.ExecuteContext)
   n2 = &RigVMFunction_Sequence(ExecuteContext=@RigVMFunction_Sequence.B)
   n3 = &RigVMFunction_Sequence(ExecuteContext=@n2.B)
   n4 = &RigVMFunction_Sequence(ExecuteContext=@n3.B)
   n5 = &RigVMFunction_Sequence(ExecuteContext=@n4.B)
   RigVMFunctionReturnNode(A=@RigVMFunction_Sequence.A, E=@n4.A, C=@n2.A, D=@n3.A, F=@n5.A, G=@n5.B)
  ]}
 RigVMModel {nodes [
   n1 = &RigUnit_PrepareForExecution()
   addFKCtrl = RigVMFunctionReferenceNode(
    Custom_FKCtrl=()
    Options_FKCtrl="(pelvis_1_#1=false,spine_3_#2=false,clavicle_5_#3=false,arm_7_#4=false,leg_9_#5=false,head_11_46DC15EF465C ...(truncated)"
    Shape_Settings="(pelvis_2_#6=(bVisible=True,Name=\"Hexagon_Thin\",Color=(R=0.0117,G=0.8281,B=1,A=1),Transform=(Rotation=(X=0,Y=-0.7071,Z=0,W=0.7071),Translation=(X=2,Y=0,Z=0),Scale3D=(X=5,Y=5,Z=7))),spine_4_#7=(bVi ...(truncated)"
    ExecuteContext=@n1.ExecutePin
   )
   RigUnit_BeginExecution = &RigUnit_BeginExecution()
   setFKCtrl = RigVMFunctionReferenceNode(setFK_Ctrl=false, ExecuteContext=@RigUnit_BeginExecution.ExecutePin)
   > FN_DeformRig_FNC_1 = RigVMFunctionReferenceNode(
    DeformRig_ Curve="(deformRig_l_upperarm_up_alpha_72_#8=true,deformRig_r_upperarm_up_alpha_73_#9=true,deformRig_l_upperarm_fr_alpha_74_#10=false,deformRig_r_upperarm_fr_alpha_75_CB51B80C ...(truncated)"
    DeformRig_Bool="(upperarm_1_#11=true,elbow_3_#12=True,wrist_7_#13=True,groin_8_#14=True,glute_17_#15=True,knee_9_40C408324CA261 ...(truncated)"
    DeformRig_Debug=(upperarm_debug_18_#11=False,wrist_debug_20_#13=False,lat_debug_26_#16=False,groinGlute_debug_27_#14=False)
    DeformRig_Value="(upperarm_up_18_#17=(Rotation=(X=0.0,Y=-0.0,Z=0.0,W=1.0),Translation=(X=0.0,Y=0.0,Z=8.0),Scale3D=(X=0.70,Y=1.0,Z=1.0)),upperarm_fwd_fr_19_#18=(Rotation=(X=0.0,Y=-0.0,Z=0.0,W=1.0),Translation=(X=0.0, ...(truncated)"
   )
   > deform_lat = RigVMFunctionReferenceNode()
   > skirt = RigVMFunctionReferenceNode()
   > biceps = RigVMFunctionReferenceNode()
   > postfix_shoulder = RigVMFunctionReferenceNode()
   > n2 = &RigUnit_SphericalPoseReader(
    ActiveRegionScaleFactors=(PositiveWidth=1,NegativeWidth=1,PositiveHeight=1,NegativeHeight=1)
    ActiveRegionSize=0.025
    Debug=(bDrawDebug=false,bDraw2D=False,bDrawLocalAxes=False,DebugScale=25,DebugSegments=20,DebugThickness=0.25)
    DriverAxis=(X=-1,Y=0,Z=0)
    DriverItem=(Type=Bone,Name="upperarm_r")
    FalloffRegionScaleFactors=(PositiveWidth=1,NegativeWidth=1,PositiveHeight=1,NegativeHeight=1)
    FalloffSize=0.1
    FlipHeightScaling=False
    FlipWidthScaling=False
    OptionalParentItem=(Type=Bone,Name="None")
    RotationOffset=(X=-115,Y=0,Z=25)
   )
   > &RigUnit_ModifyTransforms(
    ItemToModify=((Item=(Type=Bone,Name="deform_upperarm_fr_base_r"),Transform=(Rotation=(X=0,Y=0,Z=0,W=-1),Translation=(X=6,Y=0,Z=4),Scale3D=(X=1,Y=1,Z=1))))
    WeightMinimum=0
    Weight=@n2.OutputParam
   )
  ]}
]
hierarchy [
 root {loc=<0 0 0>, rot=<0 0 0 1>}
 pelvis {parent=root, loc=<0 -0.7071 93.1191>, rot=<0.0127 0.707 -0.0127 -0.707>}
 spine_01 {parent=pelvis, loc=<6.7494 -0.6388 0>, rot=<0 0 0.034 -0.9994>}
 spine_02 {parent=spine_01, loc=<7.2518 -0.035 0>, rot=<0 0 0.0238 -0.9997>}
 spine_03 {parent=spine_02, loc=<6.0104 0 0>, rot=<0 0 0.0149 -0.9999>}
 spine_04 {parent=spine_03, loc=<8.1024 0 0>, rot=<0 0 -0.0132 -0.9999>}
 spine_05 {parent=spine_04, loc=<10.0146 0 0>, rot=<0 0 -0.0015 -1>}
 neck_01 {parent=spine_05, loc=<11.3138 0.1529 0>, rot=<0 0 -0.2003 -0.9797>}
 neck_02 {parent=neck_01, loc=<4.5627 0 0>, rot=<0 0 0.0504 -0.9987>}
 head {parent=neck_02, loc=<5.3914 0 0>, rot=<0 0 0.115 -0.9934>}
 clavicle_l {parent=spine_05, loc=<4.3575 -0.1466 -2.3223>, rot=<0.0465 -0.7289 -0.009 -0.6829>}
 upperarm_l {parent=clavicle_l, loc=<15.5111 0 0>, rot=<0.0226 -0.3234 -0.0088 -0.9459>}
 lowerarm_l {parent=upperarm_l, loc=<26.6514 0 0>, rot=<0.0391 -0.0099 -0.2453 -0.9686>}
 hand_l {parent=lowerarm_l, loc=<22.3491 0 0>, rot=<-0.7081 -0.0507 0.0975 -0.6975>}
 thumb_01_l {parent=hand_l, loc=<1.3118 -0.3191 -2.459>, rot=<0.6546 -0.4042 0.0664 -0.6354>}
 thumb_02_l {parent=thumb_01_l, loc=<5.9336 -0.0448 0.0401>, rot=<-0.0037 0.0213 0.298 -0.9543>}
 thumb_03_l {parent=thumb_02_l, loc=<2.2223 0.3682 -0.0907>, rot=<0.0118 -0.013 0.0066 -0.9998>}
 index_metacarpal_l {parent=hand_l, loc=<1.6535 0.0183 -2.96>, rot=<0.0158 -0.0124 -0.0173 -0.9996>}
 index_01_l {parent=index_metacarpal_l, loc=<8.3008 0.0357 -0.0524>, rot=<-0.0044 0.0485 0.0924 -0.9945>}
 index_02_l {parent=index_01_l, loc=<3.3587 -0.3004 0.0452>, rot=<-0.002 0.0027 0.3225 -0.9466>}
 index_03_l {parent=index_02_l, loc=<2.5578 -0.123 -0.0252>, rot=<-0.002 0.0067 0.2558 -0.9667>}
 middle_metacarpal_l {parent=hand_l, loc=<2.1856 -0.1725 -0.5984>, rot=<0.006 0.0168 -0.0329 -0.9993>}
 middle_01_l {parent=middle_metacarpal_l, loc=<7.3275 -0.2447 0.0129>, rot=<-0.0208 0.0042 0.0858 -0.9961>}
 middle_02_l {parent=middle_01_l, loc=<4.0123 0.1401 0.1175>, rot=<-0.0218 0.0258 0.3512 -0.9357>}
 middle_03_l {parent=middle_02_l, loc=<2.6077 0.0534 0.1373>, rot=<0.0082 0.0188 0.1708 -0.9851>}
 ring_metacarpal_l {parent=hand_l, loc=<2.2074 -0.3625 1.3638>, rot=<-0.0284 0.0722 -0.0458 -0.9959>}
 ring_01_l {parent=ring_metacarpal_l, loc=<7.0295 -0.6197 -0.1446>, rot=<-0.0489 -0.0252 0.1892 -0.9804>}
 ring_02_l {parent=ring_01_l, loc=<3.6174 0.4647 -0.1027>, rot=<0 0 0.2806 -0.9598>}
 ring_03_l {parent=ring_02_l, loc=<2.4152 0.114 0.0199>, rot=<0 0 0.1189 -0.9929>}
 pinky_metacarpal_l {parent=hand_l, loc=<1.7616 -0.9753 3.3814>, rot=<-0.0397 0.0814 -0.069 -0.9935>}
 pinky_01_l {parent=pinky_metacarpal_l, loc=<7.1709 -0.4831 -0.1094>, rot=<-0.0654 -0.0035 0.2242 -0.9723>}
 pinky_02_l {parent=pinky_01_l, loc=<2.8274 0.0893 -0.1043>, rot=<0.0037 0.0011 0.2854 -0.9584>}
 pinky_03_l {parent=pinky_02_l, loc=<2.2046 0.0307 -0.0078>, rot=<-0.0002 0.0098 0.1445 -0.9895>}
 deform_wrist_dn_base_l {parent=hand_l, loc=<0 -2.2342 0>, rot=<0 0 0 -1>}
 deform_wrist_dn_l {parent=deform_wrist_dn_base_l, loc=<0 0 0>, rot=<0 0 0 1>}
 deform_wrist_up_base_l {parent=hand_l, loc=<0 2.2682 0>, rot=<0 0 0 -1>}
 deform_wrist_up_l {parent=deform_wrist_up_base_l, loc=<0 0 0>, rot=<0 0 0 1>}
 lowerarm_twist_02_l {parent=lowerarm_l, loc=<9.6199 0 0>, rot=<0 0 0 -1>}
 lowerarm_twist_01_l {parent=lowerarm_l, loc=<17.8549 0 0>, rot=<0 0 0 -1>}
 deform_elbow_in_base_l {parent=lowerarm_l, loc=<0.2112 5.259 0.8274>, rot=<0 0 0 -1>}
 deform_elbow_in_l {parent=deform_elbow_in_base_l, loc=<0 0 0>, rot=<0 0 0 1>}
 deform_elbow_out_base_l {parent=lowerarm_l, loc=<0 -4 0>, rot=<0 0 0 -1>}
 deform_elbow_out_l {parent=deform_elbow_out_base_l, loc=<0 0 0>, rot=<0 0 0 1>}
 upperarm_twist_01_l {parent=upperarm_l, loc=<0.0041 0 0>, rot=<0 0 0 -1>}
 deform_upperarm_bk_base_l {parent=upperarm_twist_01_l, loc=<1.0813 -7.2193 -1.9453>, rot=<0 0 0 -1>}
 deform_upperarm_bk_l {parent=deform_upperarm_bk_base_l, loc=<0 0 0>, rot=<0 0 0 1>}
 deform_upperarm_fr_base_l {parent=upperarm_twist_01_l, loc=<-0.3691 7.0506 -2.765>, rot=<0 0 0 -1>}
 deform_upperarm_fr_l {parent=deform_upperarm_fr_base_l, loc=<0 0 0>, rot=<0 0 0 1>}
 deform_upperarm_up_base_l {parent=upperarm_twist_01_l, loc=<-3.2216 -1.1391 3.3311>, rot=<0 0 0 1>}
 deform_upperarm_up_l {parent=deform_upperarm_up_base_l, loc=<0 0 0>, rot=<0 0 0 1>}
 upperarm_twist_02_l {parent=upperarm_l, loc=<13.7941 0 0>, rot=<0 0 0 -1>}
 upperarm_tricep_l_base {parent=upperarm_twist_02_l, loc=<0.5941 -7.3474 0.0268>, rot=<0 0 -0.7232 -0.6907>}
 upperarm_tricep_l {parent=upperarm_tricep_l_base, loc=<0 0 0>, rot=<0 0 0 1>}
 upperarm_bicep_l_base {parent=upperarm_twist_02_l, loc=<1.2412 6.7923 -0.0095>, rot=<0 0 -0.695 -0.719>}
 upperarm_bicep_l {parent=upperarm_bicep_l_base, loc=<0 0 0>, rot=<0 0 0 1>}
 clavicle_r {parent=spine_05, loc=<4.3575 -0.1466 2.3223>, rot=<0.7289 0.0465 -0.6829 0.009>}
 upperarm_r {parent=clavicle_r, loc=<-15.5111 0 0>, rot=<0.0226 -0.3234 -0.0088 -0.9459>}
 lowerarm_r {parent=upperarm_r, loc=<-26.6514 0 0>, rot=<0.0391 -0.0099 -0.2453 -0.9686>}
 hand_r {parent=lowerarm_r, loc=<-22.3491 0 0>, rot=<-0.7081 -0.0507 0.0975 -0.6975>}
 thumb_01_r {parent=hand_r, loc=<-1.3118 0.3191 2.459>, rot=<0.6546 -0.4042 0.0664 -0.6354>}
 thumb_02_r {parent=thumb_01_r, loc=<-5.9336 0.0448 -0.0401>, rot=<-0.0037 0.0213 0.298 -0.9543>}
 thumb_03_r {parent=thumb_02_r, loc=<-2.2223 -0.3682 0.0907>, rot=<0.0118 -0.013 0.0066 -0.9998>}
 index_metacarpal_r {parent=hand_r, loc=<-1.6535 -0.0183 2.96>, rot=<0.0158 -0.0124 -0.0173 -0.9996>}
 index_01_r {parent=index_metacarpal_r, loc=<-8.3008 -0.0357 0.0524>, rot=<-0.0044 0.0485 0.0924 -0.9945>}
 index_02_r {parent=index_01_r, loc=<-3.3587 0.3004 -0.0452>, rot=<-0.002 0.0027 0.3225 -0.9466>}
 index_03_r {parent=index_02_r, loc=<-2.5578 0.123 0.0252>, rot=<-0.002 0.0067 0.2558 -0.9667>}
 middle_metacarpal_r {parent=hand_r, loc=<-2.1856 0.1725 0.5984>, rot=<0.006 0.0168 -0.0329 -0.9993>}
 middle_01_r {parent=middle_metacarpal_r, loc=<-7.3275 0.2447 -0.0129>, rot=<-0.0208 0.0042 0.0858 -0.9961>}
 middle_02_r {parent=middle_01_r, loc=<-4.0123 -0.1401 -0.1175>, rot=<-0.0218 0.0258 0.3512 -0.9357>}
 middle_03_r {parent=middle_02_r, loc=<-2.6077 -0.0534 -0.1373>, rot=<0.0082 0.0188 0.1708 -0.9851>}
 ring_metacarpal_r {parent=hand_r, loc=<-2.2074 0.3625 -1.3638>, rot=<-0.0284 0.0722 -0.0458 -0.9959>}
 ring_01_r {parent=ring_metacarpal_r, loc=<-7.0295 0.6197 0.1446>, rot=<-0.0489 -0.0252 0.1892 -0.9804>}
 ring_02_r {parent=ring_01_r, loc=<-3.6174 -0.4647 0.1027>, rot=<0 0 0.2806 -0.9598>}
 ring_03_r {parent=ring_02_r, loc=<-2.4152 -0.114 -0.0199>, rot=<0 0 0.1189 -0.9929>}
 pinky_metacarpal_r {parent=hand_r, loc=<-1.7616 0.9753 -3.3814>, rot=<-0.0397 0.0814 -0.069 -0.9935>}
 pinky_01_r {parent=pinky_metacarpal_r, loc=<-7.1709 0.4831 0.1094>, rot=<-0.0654 -0.0035 0.2242 -0.9723>}
 pinky_02_r {parent=pinky_01_r, loc=<-2.8274 -0.0893 0.1043>, rot=<0.0037 0.0011 0.2854 -0.9584>}
 pinky_03_r {parent=pinky_02_r, loc=<-2.2046 -0.0307 0.0078>, rot=<-0.0002 0.0098 0.1445 -0.9895>}
 deform_wrist_dn_base_r {parent=hand_r, loc=<-0.0004 2.2345 -0.0001>, rot=<0 0 0 -1>}
 deform_wrist_dn_r {parent=deform_wrist_dn_base_r, loc=<0 0 0>, rot=<0 0 0 1>}
 deform_wrist_up_base_r {parent=hand_r, loc=<0.0003 -2.2684 0.0001>, rot=<0 0 0 -1>}
 deform_wrist_up_r {parent=deform_wrist_up_base_r, loc=<0 0 0>, rot=<0 0 0 1>}
 lowerarm_twist_02_r {parent=lowerarm_r, loc=<-9.6199 0 0>, rot=<0 0 0 -1>}
 lowerarm_twist_01_r {parent=lowerarm_r, loc=<-17.8549 0 0>, rot=<0 0 0 -1>}
 deform_elbow_in_base_r {parent=lowerarm_r, loc=<-0.2111 -5.2591 -0.8275>, rot=<0 0 0 -1>}
 deform_elbow_in_r {parent=deform_elbow_in_base_r, loc=<-0.0001 0 0.0002>, rot=<0 0 0 1>}
 deform_elbow_out_base_r {parent=lowerarm_r, loc=<0 4 -0.0001>, rot=<0 0 0 -1>}
 deform_elbow_out_r {parent=deform_elbow_out_base_r, loc=<0 0 0>, rot=<0 0 0 1>}
 upperarm_twist_01_r {parent=upperarm_r, loc=<-0.0041 0 0>, rot=<0 0 0 -1>}
 deform_upperarm_bk_base_r {parent=upperarm_twist_01_r, loc=<-1.0814 7.2192 1.9455>, rot=<0 0 0 1>}
 deform_upperarm_bk_r {parent=deform_upperarm_bk_base_r, loc=<0 0 0>, rot=<0 0 0 1>}
 deform_upperarm_fr_base_r {parent=upperarm_twist_01_r, loc=<0.3689 -7.0506 2.7652>, rot=<0 0 0 1>}
 deform_upperarm_fr_r {parent=deform_upperarm_fr_base_r, loc=<0 0 0>, rot=<0 0 0 1>}
 deform_upperarm_up_base_r {parent=upperarm_twist_01_r, loc=<3.2214 1.1391 -3.3308>, rot=<0 0 0 1>}
 deform_upperarm_up_r {parent=deform_upperarm_up_base_r, loc=<0 0 0>, rot=<0 0 0 -1>}
 upperarm_twist_02_r {parent=upperarm_r, loc=<-13.7941 0 0>, rot=<0 0 0 -1>}
 upperarm_bicep_r_base {parent=upperarm_twist_02_r, loc=<-1.2412 -6.7923 0.0096>, rot=<0 0 0.719 -0.695>}
 upperarm_bicep_r {parent=upperarm_bicep_r_base, loc=<0 0 0>, rot=<0 0 0 1>}
 upperarm_tricep_r_base {parent=upperarm_twist_02_r, loc=<-0.5942 7.3474 -0.0267>, rot=<0 0 0.6907 -0.7232>}
 upperarm_tricep_r {parent=upperarm_tricep_r_base, loc=<0 0 0>, rot=<0 0 0 1>}
 clavicle_l_set_rot_offset {parent=spine_05, loc=<4.3575 -0.1466 -2.3223>, rot=<0.0465 -0.7289 -0.009 -0.6829>}
 clavicle_l_set_rot {parent=clavicle_l_set_rot_offset, loc=<0 0 0>, rot=<0 0 0 1>}
 clavicle_r_set_rot_offset {parent=spine_05, loc=<4.3575 -0.1466 2.3223>, rot=<0.7289 0.0465 -0.6829 0.009>}
 clavicle_r_set_rot {parent=clavicle_r_set_rot_offset, loc=<0 0 0>, rot=<0 0 0 1>}
 attach_cape {parent=spine_05, loc=<9.017 -11.714 0>, rot=<0.0616 -0.7044 -0.0616 -0.7044>}
 cape {parent=attach_cape, type=Null, loc=<0 -4.1114 0>, rot=<0 0 0 1>}
 attach_backpack {parent=spine_05, loc=<-0.123 -13.473 0>, rot=<-0.0093 -0.707 0.0093 -0.707>}
 Backpack_BR {parent=attach_backpack, type=Null, loc=<0 -1.7337 0>, rot=<0 0 0 1>}
 deform_pec_aim_base_l {parent=spine_04, loc=<7.8732 12.6388 -2.4123>, rot=<0.0102 -0.4602 0.3979 -0.7936>}
 deform_pec_aim_l {parent=deform_pec_aim_base_l, loc=<0 0 0>, rot=<0 0 0 1>}
 deform_pec_base_l {parent=deform_pec_aim_l, loc=<10.766 0 0>, rot=<0 0 0 1>}
 deform_pec_l {parent=deform_pec_base_l, loc=<0.055 0 0>, rot=<0 0 0 1>}
 deform_pec_aim_base_r {parent=spine_04, loc=<7.8727 12.6388 2.4123>, rot=<-0.4602 -0.0102 0.7936 0.3979>}
 deform_pec_aim_r {parent=deform_pec_aim_base_r, loc=<0 0 0>, rot=<0 0 0 1>}
 deform_pec_base_r {parent=deform_pec_aim_r, loc=<-10.7662 -0.0006 -0.0007>, rot=<0 0 0 1>}
 deform_pec_r {parent=deform_pec_base_r, loc=<0 0 0>, rot=<0 0 0 1>}
 deform_lat_l_base {parent=spine_03, loc=<5.609 0 -12.5868>, rot=<0 0 0 1>}
 deform_lat_l {parent=deform_lat_l_base, loc=<0 0 0>, rot=<0 0 0 1>}
 deform_latr_r_base {parent=spine_03, loc=<5.609 0 12.587>, rot=<0 0 0 1>}
 deform_lat_r {parent=deform_latr_r_base, loc=<0 0 0>, rot=<0 0 0 1>}
 thigh_l {parent=pelvis, loc=<-2.5693 -0.0109 -7.7501>, rot=<0.0012 0.0572 0.0155 -0.9982>}
 calf_l {parent=thigh_l, loc=<-39.0484 0 0>, rot=<0.0006 -0.0161 -0.0391 -0.9991>}
 foot_l {parent=calf_l, loc=<-42.561 0.1077 0.4292>, rot=<0.0418 -0.0455 0.0343 -0.9975>}
 ball_l {parent=foot_l, loc=<-6.1599 13.8264 0.6806>, rot=<0.0001 -0.0001 -0.7063 -0.7079>}
 calf_twist_02_l {parent=calf_l, loc=<-18.8064 -0.0967 0.0397>, rot=<0.0237 0.0095 -0.0035 -0.9997>}
 calf_twist_01_l {parent=calf_l, loc=<-35.813 -0.0967 0.0397>, rot=<0.0237 0.0095 -0.0035 -0.9997>}
 deform_knee_in_base_l {parent=calf_l, loc=<1.298 -5.2957 -0.0496>, rot=<0 0 0 -1>}
 deform_knee_in_l {parent=deform_knee_in_base_l, loc=<0 0 0>, rot=<0 0 0 1>}
 deform_knee_out_base_l {parent=calf_l, loc=<0 4 0>, rot=<0 0 0 -1>}
 deform_knee_out_l {parent=deform_knee_out_base_l, loc=<0 0 0>, rot=<0 0 0 1>}
 thigh_twist_01_l {parent=thigh_l, loc=<-21.6303 0 0>, rot=<0.0013 0 0 -1>}
 deform_groin_base_l {parent=thigh_l, loc=<-2.1908 8.1142 -1.7442>, rot=<0 0 0 -1>}
 deform_groin_l {parent=deform_groin_base_l, loc=<0 0 0>, rot=<0 0 0 1>}
 thigh_r {parent=pelvis, loc=<-2.5693 -0.0109 7.7501>, rot=<0.0572 -0.0012 0.9982 0.0155>}
 calf_r {parent=thigh_r, loc=<39.0484 0 0>, rot=<0.0006 -0.0161 -0.0391 -0.9991>}
 foot_r {parent=calf_r, loc=<42.561 -0.1077 -0.4292>, rot=<0.0418 -0.0455 0.0343 -0.9975>}
 ball_r {parent=foot_r, loc=<6.1599 -13.8264 -0.6806>, rot=<0.0001 -0.0001 -0.7063 -0.7079>}
 calf_twist_02_r {parent=calf_r, loc=<18.8064 0.0967 -0.0397>, rot=<0.0237 0.0095 -0.0035 -0.9997>}
 calf_twist_01_r {parent=calf_r, loc=<35.813 0.0967 -0.0397>, rot=<0.0237 0.0095 -0.0035 -0.9997>}
 deform_knee_in_base_r {parent=calf_r, loc=<-1.298 5.2957 0.0496>, rot=<0 0 0 -1>}
 deform_knee_in_r {parent=deform_knee_in_base_r, loc=<0 0 0>, rot=<0 0 0 1>}
 deform_knee_out_base_r {parent=calf_r, loc=<0 -4 0>, rot=<0 0 0 -1>}
 deform_knee_out_r {parent=deform_knee_out_base_r, loc=<0 0 0>, rot=<0 0 0 1>}
 thigh_twist_01_r {parent=thigh_r, loc=<21.6303 0 0>, rot=<0.0013 0 0 -1>}
 deform_groin_base_r {parent=thigh_r, loc=<2.1908 -8.1142 1.7442>, rot=<0 0 0 -1>}
 deform_groin_r {parent=deform_groin_base_r, loc=<0 0 0>, rot=<0 0 0 1>}
 deform_glute_base_l {parent=pelvis, loc=<-8.8064 -8.8222 -8.4287>, rot=<0.0275 0.0501 -0.4472 -0.8926>}
 deform_glute_l {parent=deform_glute_base_l, loc=<0 0 0>, rot=<0 0 0 1>}
 deform_glute_base_r {parent=pelvis, loc=<-8.8062 -8.8222 8.4287>, rot=<0.0501 -0.0275 0.8926 -0.4472>}
 deform_glute_r {parent=deform_glute_base_r, loc=<0 0 0>, rot=<0 0 0 1>}
 base_skirt {parent=pelvis, loc=<0 0 0>, rot=<0 0 0 1>}
 base_skirt_front_0 {parent=base_skirt, loc=<3.0388 9.4116 0>, rot=<0 0 0.9987 0.0518>}
 dyn_skirt_front_0 {parent=base_skirt_front_0, loc=<0 0 0>, rot=<0 0 0 1>}
 dyn_skirt_front_1 {parent=dyn_skirt_front_0, loc=<11.1537 0 0>, rot=<0 0 -0.0175 -0.9998>}
 dyn_skirt_front_2 {parent=dyn_skirt_front_1, loc=<10.8946 0 0>, rot=<0 0 -0.0872 -0.9962>}
 base_skirt_front_l_0 {parent=base_skirt, loc=<3.1268 6.8972 -8.9788>, rot=<-0.093 0.2424 0.9609 0.0959>}
 dyn_skirt_front_l_0 {parent=base_skirt_front_l_0, loc=<0 0 0>, rot=<0 0 0 1>}
 dyn_skirt_front_l_1 {parent=dyn_skirt_front_l_0, loc=<10.9 0 0>, rot=<0 0 0 1>}
 dyn_skirt_front_l_2 {parent=dyn_skirt_front_l_1, loc=<10.9 0 0>, rot=<0 -0.0151 -0.0858 -0.9962>}
 base_skirt_side_l_0 {parent=base_skirt, loc=<2.9515 -0.5891 -13.3655>, rot=<0.1325 -0.6987 -0.6904 -0.1329>}
 dyn_skirt_side_l_0 {parent=base_skirt_side_l_0, loc=<0 0 0>, rot=<0 0 0 1>}
 dyn_skirt_side_l_1 {parent=dyn_skirt_side_l_0, loc=<10.9 0 0>, rot=<0 0 -0.0245 -0.9997>}
 dyn_skirt_side_l_2 {parent=dyn_skirt_side_l_1, loc=<10.9 0 0>, rot=<0 0 -0.0872 -0.9962>}
 base_skirt_back_side_l_0 {parent=base_skirt, loc=<3.8118 -10.678 -9.4235>, rot=<0.1248 -0.9474 -0.2851 -0.0739>}
 dyn_skirt_back_side_l_0 {parent=base_skirt_back_side_l_0, loc=<0 0 0>, rot=<0 0 0 1>}
 dyn_skirt_back_side_l_1 {parent=dyn_skirt_back_side_l_0, loc=<10.9 0 0>, rot=<0 0 -0.0877 -0.9961>}
 dyn_skirt_back_side_l_2 {parent=dyn_skirt_back_side_l_1, loc=<10.9 0 0>, rot=<0 0 -0.0872 -0.9962>}
 base_skirt_back_0 {parent=base_skirt, loc=<2.8666 -12.2381 0>, rot=<0.1127 -0.9936 0 0>}
 dyn_skirt_back_0 {parent=base_skirt_back_0, loc=<0 0 0>, rot=<0 0 0.0392 -0.9992>}
 dyn_skirt_back_1 {parent=dyn_skirt_back_0, loc=<10.9 0 0>, rot=<0 0 -0.1355 -0.9908>}
 dyn_skirt_back_2 {parent=dyn_skirt_back_1, loc=<10.9 0 0>, rot=<0 0 -0.0262 -0.9997>}
 base_skirt_back_side_r_0 {parent=base_skirt, loc=<2.9592 -10.8007 9.6192>, rot=<0.1098 -0.9493 0.2863 0.0694>}
 dyn_skirt_back_side_r_0 {parent=base_skirt_back_side_r_0, loc=<0 0 0>, rot=<0 0 0 1>}
 dyn_skirt_back_side_r_1 {parent=dyn_skirt_back_side_r_0, loc=<10.9 0 0>, rot=<0 0 -0.0641 -0.9979>}
 dyn_skirt_back_side_r_2 {parent=dyn_skirt_back_side_r_1, loc=<10.9 0 0.0001>, rot=<0 0 -0.0872 -0.9962>}
 base_skirt_side_r_0 {parent=base_skirt, loc=<2.8303 -0.5863 13.153>, rot=<0.1344 -0.6983 0.69 0.1349>}
 dyn_skirt_side_r_0 {parent=base_skirt_side_r_0, loc=<0 0 0>, rot=<0 0 0 1>}
 dyn_skirt_side_r_1 {parent=dyn_skirt_side_r_0, loc=<10.8997 -0.0002 0>, rot=<0 0 -0.0205 -0.9998>}
 dyn_skirt_side_r_2 {parent=dyn_skirt_side_r_1, loc=<10.9 0 0>, rot=<0 0 -0.0872 -0.9962>}
 base_skirt_front_r_0 {parent=base_skirt, loc=<3.1396 6.936 8.9986>, rot=<0.0939 -0.242 0.9606 0.0995>}
 dyn_skirt_front_r_0 {parent=base_skirt_front_r_0, loc=<0 0 0>, rot=<0 0 0 1>}
 dyn_skirt_front_r_1 {parent=dyn_skirt_front_r_0, loc=<10.9 0 0>, rot=<0 0 0 1>}
 dyn_skirt_front_r_2 {parent=dyn_skirt_front_r_1, loc=<10.9001 0 0>, rot=<0 0.0151 -0.0858 -0.9962>}
 CTRL_expressions_browDownL {type=Curve, loc=<0 0 0>, rot=<0 0 0 1>}
 CTRL_expressions_browDownR {type=Curve, loc=<0 0 0>, rot=<0 0 0 1>}
 CTRL_expressions_browLateralL {type=Curve, loc=<0 0 0>, rot=<0 0 0 1>}
 CTRL_expressions_browLateralR {type=Curve, loc=<0 0 0>, rot=<0 0 0 1>}
 CTRL_expressions_browRaiseInL {type=Curve, loc=<0 0 0>, rot=<0 0 0 1>}
 CTRL_expressions_browRaiseInR {type=Curve, loc=<0 0 0>, rot=<0 0 0 1>}
 CTRL_expressions_browRaiseOuterL {type=Curve, loc=<0 0 0>, rot=<0 0 0 1>}
 CTRL_expressions_browRaiseOuterR {type=Curve, loc=<0 0 0>, rot=<0 0 0 1>}
 CTRL_expressions_browUINL {type=Curve, loc=<0 0 0>, rot=<0 0 0 1>}
 CTRL_expressions_browUINR {type=Curve, loc=<0 0 0>, rot=<0 0 0 1>}
 CTRL_expressions_browUOUTL {type=Curve, loc=<0 0 0>, rot=<0 0 0 1>}
 CTRL_expressions_browUOUTR {type=Curve, loc=<0 0 0>, rot=<0 0 0 1>}
 CTRL_expressions_browDINL {type=Curve, loc=<0 0 0>, rot=<0 0 0 1>}
 CTRL_expressions_browDINR {type=Curve, loc=<0 0 0>, rot=<0 0 0 1>}
 CTRL_expressions_browDOUTL {type=Curve, loc=<0 0 0>, rot=<0 0 0 1>}
 CTRL_expressions_browDOUTR {type=Curve, loc=<0 0 0>, rot=<0 0 0 1>}
 CTRL_expressions_eyeBlinkL {type=Curve, loc=<0 0 0>, rot=<0 0 0 1>}
 CTRL_expressions_eyeBlinkR {type=Curve, loc=<0 0 0>, rot=<0 0 0 1>}
 CTRL_expressions_eyeWidenL {type=Curve, loc=<0 0 0>, rot=<0 0 0 1>}
 CTRL_expressions_eyeWidenR {type=Curve, loc=<0 0 0>, rot=<0 0 0 1>}
 CTRL_expressions_eyeSquintInnerL {type=Curve, loc=<0 0 0>, rot=<0 0 0 1>}
 CTRL_expressions_eyeSquintInnerR {type=Curve, loc=<0 0 0>, rot=<0 0 0 1>}
 CTRL_expressions_eyeCheekRaiseL {type=Curve, loc=<0 0 0>, rot=<0 0 0 1>}
 CTRL_expressions_eyeCheekRaiseR {type=Curve, loc=<0 0 0>, rot=<0 0 0 1>}
 CTRL_expressions_eyeUpperLidUpL {type=Curve, loc=<0 0 0>, rot=<0 0 0 1>}
 CTRL_expressions_eyeUpperLidUpR {type=Curve, loc=<0 0 0>, rot=<0 0 0 1>}
 CTRL_expressions_eyeRelaxL {type=Curve, loc=<0 0 0>, rot=<0 0 0 1>}
 CTRL_expressions_eyeRelaxR {type=Curve, loc=<0 0 0>, rot=<0 0 0 1>}
 CTRL_expressions_eyeLowerLidUpL {type=Curve, loc=<0 0 0>, rot=<0 0 0 1>}
 CTRL_expressions_eyeLowerLidUpR {type=Curve, loc=<0 0 0>, rot=<0 0 0 1>}
 CTRL_expressions_eyeLowerLidDownL {type=Curve, loc=<0 0 0>, rot=<0 0 0 1>}
 CTRL_expressions_eyeLowerLidDownR {type=Curve, loc=<0 0 0>, rot=<0 0 0 1>}
 CTRL_expressions_eyelidRollPosUL {type=Curve, loc=<0 0 0>, rot=<0 0 0 1>}
 CTRL_expressions_eyelidRollPosUR {type=Curve, loc=<0 0 0>, rot=<0 0 0 1>}
 CTRL_expressions_eyelidRollPosDL {type=Curve, loc=<0 0 0>, rot=<0 0 0 1>}
 CTRL_expressions_eyelidRollPosDR {type=Curve, loc=<0 0 0>, rot=<0 0 0 1>}
 CTRL_expressions_eyelidRollNegUL {type=Curve, loc=<0 0 0>, rot=<0 0 0 1>}
 CTRL_expressions_eyelidRollNegUR {type=Curve, loc=<0 0 0>, rot=<0 0 0 1>}
 CTRL_expressions_eyelidRollNegDL {type=Curve, loc=<0 0 0>, rot=<0 0 0 1>}
 CTRL_expressions_eyelidRollNegDR {type=Curve, loc=<0 0 0>, rot=<0 0 0 1>}
 CTRL_expressions_eyeLookUpL {type=Curve, loc=<0 0 0>, rot=<0 0 0 1>}
 CTRL_expressions_eyeLookUpR {type=Curve, loc=<0 0 0>, rot=<0 0 0 1>}
 CTRL_expressions_eyeLookDownL {type=Curve, loc=<0 0 0>, rot=<0 0 0 1>}
 CTRL_expressions_eyeLookDownR {type=Curve, loc=<0 0 0>, rot=<0 0 0 1>}
 CTRL_expressions_eyeLookLeftL {type=Curve, loc=<0 0 0>, rot=<0 0 0 1>}
 CTRL_expressions_eyeLookLeftR {type=Curve, loc=<0 0 0>, rot=<0 0 0 1>}
 CTRL_expressions_eyeLookRightL {type=Curve, loc=<0 0 0>, rot=<0 0 0 1>}
 CTRL_expressions_eyeLookRightR {type=Curve, loc=<0 0 0>, rot=<0 0 0 1>}
 CTRL_expressions_eyeEyelidsReduce {type=Curve, loc=<0 0 0>, rot=<0 0 0 1>}
 CTRL_expressions_noseWrinkleL {type=Curve, loc=<0 0 0>, rot=<0 0 0 1>}
 CTRL_expressions_noseWrinkleR {type=Curve, loc=<0 0 0>, rot=<0 0 0 1>}
 CTRL_expressions_noseWrinkleUpperL {type=Curve, loc=<0 0 0>, rot=<0 0 0 1>}
 CTRL_expressions_noseWrinkleUpperR {type=Curve, loc=<0 0 0>, rot=<0 0 0 1>}
 CTRL_expressions_noseNostrilDepressL {type=Curve, loc=<0 0 0>, rot=<0 0 0 1>}
 CTRL_expressions_noseNostrilDepressR {type=Curve, loc=<0 0 0>, rot=<0 0 0 1>}
 CTRL_expressions_noseNostrilDilateL {type=Curve, loc=<0 0 0>, rot=<0 0 0 1>}
 CTRL_expressions_noseNostrilDilateR {type=Curve, loc=<0 0 0>, rot=<0 0 0 1>}
 CTRL_expressions_noseNostrilCompressL {type=Curve, loc=<0 0 0>, rot=<0 0 0 1>}
 CTRL_expressions_noseNostrilCompressR {type=Curve, loc=<0 0 0>, rot=<0 0 0 1>}
 CTRL_expressions_noseRotateL {type=Curve, loc=<0 0 0>, rot=<0 0 0 1>}
 CTRL_expressions_noseRotateR {type=Curve, loc=<0 0 0>, rot=<0 0 0 1>}
 CTRL_expressions_noseSquash {type=Curve, loc=<0 0 0>, rot=<0 0 0 1>}
 CTRL_expressions_noseStretch {type=Curve, loc=<0 0 0>, rot=<0 0 0 1>}
 CTRL_expressions_mouthCheekSuckL {type=Curve, loc=<0 0 0>, rot=<0 0 0 1>}
 CTRL_expressions_mouthCheekSuckR {type=Curve, loc=<0 0 0>, rot=<0 0 0 1>}
 CTRL_expressions_mouthCheekBlowL {type=Curve, loc=<0 0 0>, rot=<0 0 0 1>}
 CTRL_expressions_mouthCheekBlowR {type=Curve, loc=<0 0 0>, rot=<0 0 0 1>}
 CTRL_expressions_mouthLipsBlowL {type=Curve, loc=<0 0 0>, rot=<0 0 0 1>}
 CTRL_expressions_mouthLipsBlowR {type=Curve, loc=<0 0 0>, rot=<0 0 0 1>}
 CTRL_expressions_mouthLeft {type=Curve, loc=<0 0 0>, rot=<0 0 0 1>}
 CTRL_expressions_mouthRight {type=Curve, loc=<0 0 0>, rot=<0 0 0 1>}
 CTRL_expressions_mouthUp {type=Curve, loc=<0 0 0>, rot=<0 0 0 1>}
 CTRL_expressions_mouthDown {type=Curve, loc=<0 0 0>, rot=<0 0 0 1>}
 CTRL_expressions_mouthUpperLipRaiseL {type=Curve, loc=<0 0 0>, rot=<0 0 0 1>}
 CTRL_expressions_mouthUpperLipRaiseR {type=Curve, loc=<0 0 0>, rot=<0 0 0 1>}
 CTRL_expressions_mouthLowerLipDepressL {type=Curve, loc=<0 0 0>, rot=<0 0 0 1>}
 CTRL_expressions_mouthLowerLipDepressR {type=Curve, loc=<0 0 0>, rot=<0 0 0 1>}
 CTRL_expressions_mouthCornerPullL {type=Curve, loc=<0 0 0>, rot=<0 0 0 1>}
 CTRL_expressions_mouthCornerPullR {type=Curve, loc=<0 0 0>, rot=<0 0 0 1>}
 CTRL_expressions_mouthStretchL {type=Curve, loc=<0 0 0>, rot=<0 0 0 1>}
 CTRL_expressions_mouthStretchR {type=Curve, loc=<0 0 0>, rot=<0 0 0 1>}
 CTRL_expressions_mouthDimpleL {type=Curve, loc=<0 0 0>, rot=<0 0 0 1>}
 CTRL_expressions_mouthDimpleR {type=Curve, loc=<0 0 0>, rot=<0 0 0 1>}
 CTRL_expressions_mouthCornerDepressL {type=Curve, loc=<0 0 0>, rot=<0 0 0 1>}
 CTRL_expressions_mouthCornerDepressR {type=Curve, loc=<0 0 0>, rot=<0 0 0 1>}
 CTRL_expressions_mouthPressUL {type=Curve, loc=<0 0 0>, rot=<0 0 0 1>}
 CTRL_expressions_mouthPressUR {type=Curve, loc=<0 0 0>, rot=<0 0 0 1>}
 CTRL_expressions_mouthPressDL {type=Curve, loc=<0 0 0>, rot=<0 0 0 1>}
 CTRL_expressions_mouthPressDR {type=Curve, loc=<0 0 0>, rot=<0 0 0 1>}
 CTRL_expressions_mouthLipsPurseUL {type=Curve, loc=<0 0 0>, rot=<0 0 0 1>}
 CTRL_expressions_mouthLipsPurseUR {type=Curve, loc=<0 0 0>, rot=<0 0 0 1>}
 CTRL_expressions_mouthLipsPurseDL {type=Curve, loc=<0 0 0>, rot=<0 0 0 1>}
 CTRL_expressions_mouthLipsPurseDR {type=Curve, loc=<0 0 0>, rot=<0 0 0 1>}
 CTRL_expressions_mouthLipsTowardsUL {type=Curve, loc=<0 0 0>, rot=<0 0 0 1>}
 CTRL_expressions_mouthLipsTowardsUR {type=Curve, loc=<0 0 0>, rot=<0 0 0 1>}
 CTRL_expressions_mouthLipsTowardsDL {type=Curve, loc=<0 0 0>, rot=<0 0 0 1>}
 CTRL_expressions_mouthLipsTowardsDR {type=Curve, loc=<0 0 0>, rot=<0 0 0 1>}
 CTRL_expressions_mouthFunnelUL {type=Curve, loc=<0 0 0>, rot=<0 0 0 1>}
 CTRL_expressions_mouthFunnelUR {type=Curve, loc=<0 0 0>, rot=<0 0 0 1>}
 CTRL_expressions_mouthFunnelDL {type=Curve, loc=<0 0 0>, rot=<0 0 0 1>}
 CTRL_expressions_mouthFunnelDR {type=Curve, loc=<0 0 0>, rot=<0 0 0 1>}
 CTRL_expressions_mouthLipsTogetherUL {type=Curve, loc=<0 0 0>, rot=<0 0 0 1>}
 CTRL_expressions_mouthLipsTogetherUR {type=Curve, loc=<0 0 0>, rot=<0 0 0 1>}
 CTRL_expressions_mouthLipsTogetherDL {type=Curve, loc=<0 0 0>, rot=<0 0 0 1>}
 CTRL_expressions_mouthLipsTogetherDR {type=Curve, loc=<0 0 0>, rot=<0 0 0 1>}
 CTRL_expressions_mouthUpperLipBiteL {type=Curve, loc=<0 0 0>, rot=<0 0 0 1>}
 CTRL_expressions_mouthUpperLipBiteR {type=Curve, loc=<0 0 0>, rot=<0 0 0 1>}
 CTRL_expressions_mouthLowerLipBiteL {type=Curve, loc=<0 0 0>, rot=<0 0 0 1>}
 CTRL_expressions_mouthLowerLipBiteR {type=Curve, loc=<0 0 0>, rot=<0 0 0 1>}
 CTRL_expressions_mouthCornerSharpenUL {type=Curve, loc=<0 0 0>, rot=<0 0 0 1>}
 CTRL_expressions_mouthCornerSharpenUR {type=Curve, loc=<0 0 0>, rot=<0 0 0 1>}
 CTRL_expressions_mouthCornerSharpenDL {type=Curve, loc=<0 0 0>, rot=<0 0 0 1>}
 CTRL_expressions_mouthCornerSharpenDR {type=Curve, loc=<0 0 0>, rot=<0 0 0 1>}
 CTRL_expressions_mouthCornerRounderUL {type=Curve, loc=<0 0 0>, rot=<0 0 0 1>}
 CTRL_expressions_mouthCornerRounderUR {type=Curve, loc=<0 0 0>, rot=<0 0 0 1>}
 CTRL_expressions_mouthCornerRounderDL {type=Curve, loc=<0 0 0>, rot=<0 0 0 1>}
 CTRL_expressions_mouthCornerRounderDR {type=Curve, loc=<0 0 0>, rot=<0 0 0 1>}
 CTRL_expressions_mouthUpperLipTowardsTeethL {type=Curve, loc=<0 0 0>, rot=<0 0 0 1>}
 CTRL_expressions_mouthUpperLipTowardsTeethR {type=Curve, loc=<0 0 0>, rot=<0 0 0 1>}
 CTRL_expressions_mouthLowerLipTowardsTeethL {type=Curve, loc=<0 0 0>, rot=<0 0 0 1>}
 CTRL_expressions_mouthLowerLipTowardsTeethR {type=Curve, loc=<0 0 0>, rot=<0 0 0 1>}
 CTRL_expressions_mouthCornerUpL {type=Curve, loc=<0 0 0>, rot=<0 0 0 1>}
 CTRL_expressions_mouthCornerUpR {type=Curve, loc=<0 0 0>, rot=<0 0 0 1>}
 CTRL_expressions_mouthCornerDownL {type=Curve, loc=<0 0 0>, rot=<0 0 0 1>}
 CTRL_expressions_mouthCornerDownR {type=Curve, loc=<0 0 0>, rot=<0 0 0 1>}
 CTRL_expressions_mouthCornerWideL {type=Curve, loc=<0 0 0>, rot=<0 0 0 1>}
 CTRL_expressions_mouthCornerWideR {type=Curve, loc=<0 0 0>, rot=<0 0 0 1>}
 CTRL_expressions_mouthCornerNarrowL {type=Curve, loc=<0 0 0>, rot=<0 0 0 1>}
 CTRL_expressions_mouthCornerNarrowR {type=Curve, loc=<0 0 0>, rot=<0 0 0 1>}
 CTRL_expressions_tongueUp {type=Curve, loc=<0 0 0>, rot=<0 0 0 1>}
 CTRL_expressions_tongueDown {type=Curve, loc=<0 0 0>, rot=<0 0 0 1>}
 CTRL_expressions_tongueLeft {type=Curve, loc=<0 0 0>, rot=<0 0 0 1>}
 CTRL_expressions_tongueRight {type=Curve, loc=<0 0 0>, rot=<0 0 0 1>}
 CTRL_expressions_tongueOut {type=Curve, loc=<0 0 0>, rot=<0 0 0 1>}
 CTRL_expressions_tongueIn {type=Curve, loc=<0 0 0>, rot=<0 0 0 1>}
 CTRL_expressions_tongueRollUp {type=Curve, loc=<0 0 0>, rot=<0 0 0 1>}
 CTRL_expressions_tongueRollDown {type=Curve, loc=<0 0 0>, rot=<0 0 0 1>}
 CTRL_expressions_tongueRollLeft {type=Curve, loc=<0 0 0>, rot=<0 0 0 1>}
 CTRL_expressions_tongueRollRight {type=Curve, loc=<0 0 0>, rot=<0 0 0 1>}
 CTRL_expressions_tongueWide {type=Curve, loc=<0 0 0>, rot=<0 0 0 1>}
 CTRL_expressions_tongueNarrow {type=Curve, loc=<0 0 0>, rot=<0 0 0 1>}
 CTRL_expressions_jawOpen {type=Curve, loc=<0 0 0>, rot=<0 0 0 1>}
 CTRL_expressions_jawLeft {type=Curve, loc=<0 0 0>, rot=<0 0 0 1>}
 CTRL_expressions_jawRight {type=Curve, loc=<0 0 0>, rot=<0 0 0 1>}
 CTRL_expressions_jawFwd {type=Curve, loc=<0 0 0>, rot=<0 0 0 1>}
 CTRL_expressions_jawBack {type=Curve, loc=<0 0 0>, rot=<0 0 0 1>}
 CTRL_expressions_jawChinRaiseDL {type=Curve, loc=<0 0 0>, rot=<0 0 0 1>}
 CTRL_expressions_jawChinRaiseDR {type=Curve, loc=<0 0 0>, rot=<0 0 0 1>}
 CTRL_expressions_jawChinRaiseUL {type=Curve, loc=<0 0 0>, rot=<0 0 0 1>}
 CTRL_expressions_jawChinRaiseUR {type=Curve, loc=<0 0 0>, rot=<0 0 0 1>}
 CTRL_expressions_jawOpenExtreme {type=Curve, loc=<0 0 0>, rot=<0 0 0 1>}
 CTRL_expressions_teethUpU {type=Curve, loc=<0 0 0>, rot=<0 0 0 1>}
 CTRL_expressions_teethUpD {type=Curve, loc=<0 0 0>, rot=<0 0 0 1>}
 CTRL_expressions_teethDownU {type=Curve, loc=<0 0 0>, rot=<0 0 0 1>}
 CTRL_expressions_teethDownD {type=Curve, loc=<0 0 0>, rot=<0 0 0 1>}
 CTRL_expressions_teethLeftU {type=Curve, loc=<0 0 0>, rot=<0 0 0 1>}
 CTRL_expressions_teethLeftD {type=Curve, loc=<0 0 0>, rot=<0 0 0 1>}
 CTRL_expressions_teethRightU {type=Curve, loc=<0 0 0>, rot=<0 0 0 1>}
 CTRL_expressions_teethRightD {type=Curve, loc=<0 0 0>, rot=<0 0 0 1>}
 CTRL_expressions_teethFwdU {type=Curve, loc=<0 0 0>, rot=<0 0 0 1>}
 CTRL_expressions_teethFwdD {type=Curve, loc=<0 0 0>, rot=<0 0 0 1>}
 CTRL_expressions_teethBackU {type=Curve, loc=<0 0 0>, rot=<0 0 0 1>}
 CTRL_expressions_teethBackD {type=Curve, loc=<0 0 0>, rot=<0 0 0 1>}
 CTRL_expressions_teethRotateFwdU {type=Curve, loc=<0 0 0>, rot=<0 0 0 1>}
 CTRL_expressions_teethRotateFwdD {type=Curve, loc=<0 0 0>, rot=<0 0 0 1>}
 CTRL_expressions_teethRotateBackU {type=Curve, loc=<0 0 0>, rot=<0 0 0 1>}
 CTRL_expressions_teethRotateBackD {type=Curve, loc=<0 0 0>, rot=<0 0 0 1>}
 DisableFaceOverride {type=Curve, loc=<0 0 0>, rot=<0 0 0 1>}
 is_3L {type=Curve, loc=<0 0 0>, rot=<0 0 0 1>}
 is_face_bone_anim {type=Curve, loc=<0 0 0>, rot=<0 0 0 1>}
 WM_Smile_L {type=Curve, loc=<0 0 0>, rot=<0 0 0 1>}
 WM_BrowLateral_L {type=Curve, loc=<0 0 0>, rot=<0 0 0 1>}
 WM_BrowLateral_R {type=Curve, loc=<0 0 0>, rot=<0 0 0 1>}
 WM_BrowRaiseInner_L {type=Curve, loc=<0 0 0>, rot=<0 0 0 1>}
 WM_BrowRaiseInner_R {type=Curve, loc=<0 0 0>, rot=<0 0 0 1>}
 WM_BrowRaiseOutter_L {type=Curve, loc=<0 0 0>, rot=<0 0 0 1>}
 WM_BrowRaiseOutter_R {type=Curve, loc=<0 0 0>, rot=<0 0 0 1>}
 WM_CheekRaise_L {type=Curve, loc=<0 0 0>, rot=<0 0 0 1>}
 WM_CheekRaise_R {type=Curve, loc=<0 0 0>, rot=<0 0 0 1>}
 WM_Lips_L {type=Curve, loc=<0 0 0>, rot=<0 0 0 1>}
 WM_Lips_R {type=Curve, loc=<0 0 0>, rot=<0 0 0 1>}
 WM_MouthStretch_L {type=Curve, loc=<0 0 0>, rot=<0 0 0 1>}
 WM_MouthStretch_R {type=Curve, loc=<0 0 0>, rot=<0 0 0 1>}
 WM_NoseWrinkle_L {type=Curve, loc=<0 0 0>, rot=<0 0 0 1>}
 WM_NoseWrinkle_R {type=Curve, loc=<0 0 0>, rot=<0 0 0 1>}
 WM_Smile_R {type=Curve, loc=<0 0 0>, rot=<0 0 0 1>}
 WM_BrowLateral_Offset_L {type=Curve, loc=<0 0 0>, rot=<0 0 0 1>}
 WM_BrowLateral_Offset_R {type=Curve, loc=<0 0 0>, rot=<0 0 0 1>}
 WM_BrowRaiseInner_Offset_L {type=Curve, loc=<0 0 0>, rot=<0 0 0 1>}
 WM_BrowRaiseInner_Offset_R {type=Curve, loc=<0 0 0>, rot=<0 0 0 1>}
 WM_BrowRaiseOutter_Offset_L {type=Curve, loc=<0 0 0>, rot=<0 0 0 1>}
 WM_BrowRaiseOutter_Offset_R {type=Curve, loc=<0 0 0>, rot=<0 0 0 1>}
 WM_CheekRaise_Offset_L {type=Curve, loc=<0 0 0>, rot=<0 0 0 1>}
 WM_CheekRaise_Offset_R {type=Curve, loc=<0 0 0>, rot=<0 0 0 1>}
 WM_Lips_Offset_L {type=Curve, loc=<0 0 0>, rot=<0 0 0 1>}
 WM_Lips_Offset_R {type=Curve, loc=<0 0 0>, rot=<0 0 0 1>}
 WM_MouthStretch_Offset_L {type=Curve, loc=<0 0 0>, rot=<0 0 0 1>}
 WM_MouthStretch_Offset_R {type=Curve, loc=<0 0 0>, rot=<0 0 0 1>}
 WM_Smile_Offset_L {type=Curve, loc=<0 0 0>, rot=<0 0 0 1>}
 WM_Smile_Offset_R {type=Curve, loc=<0 0 0>, rot=<0 0 0 1>}
 WM_NoseWrinkle_Offset_L {type=Curve, loc=<0 0 0>, rot=<0 0 0 1>}
 WM_NoseWrinkle_Offset_R {type=Curve, loc=<0 0 0>, rot=<0 0 0 1>}
]
variables [
 CrouchWalk {type=double, default=0}
 GetCrouchValue {type=double, default=0}
 IsCrouching {type=bool, default=False}
 IsSliding {type=bool, default=False}
 "Istactical " {type=bool, default=False}
 isDBNO {type=bool, default=False}
]
