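# Physics asset for the left-side skirt of the M_MED_ParadeLotus male player cosmetic body.
# Front, side and back-side chains each link three dynamic capsule bodies to a tiny base-bone anchor.
# Constraints give each joint a narrow swing1 limit and zero swing2 and twist, hinge-like, with the back-side chain allowed the widest swing.
# A large left-thigh capsule has no constraints, apparently serving as the leg collider.
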PhysicsAsset /BRCosmetics/Characters/Player/Male/Medium/Bodies/M_MED_ParadeLotus/Meshes/M_MED_ParadeLotus_Physics_Skirt
bodies {
 base_skirt_back_side_l_0 [{shapeType=Capsule, length=0.505, radius=0.505, shapeName=base_skirt_back_side_l_0_capsule, rotation=<rot 0 -90 0>}]
 base_skirt_front_l_0 [{shapeType=Capsule, length=0.505, radius=0.505, shapeName=base_skirt_front_l_0_capsule, rotation=<rot 0 -90 0>}]
 base_skirt_side_l_0 [{shapeType=Capsule, length=0.505, radius=0.505, shapeName=base_skirt_side_l_0_capsule, rotation=<rot 0 -90 0>}]
 dyn_skirt_back_side_l_0 [{shapeType=Capsule, length=12, radius=3.5, shapeName=dyn_skirt_back_side_l_0_capsule, center=<2 -3 0>}]
 dyn_skirt_back_side_l_1 [{
   shapeType=Capsule
   length=9
   radius=6.5
   shapeName=dyn_skirt_back_side_l_1_capsule
   center=<2 -3.5 0>
   rotation=<rot -3.0266 0.0002 -15.3497>
  }]
 dyn_skirt_back_side_l_2 [{shapeType=Capsule, length=11, radius=5, shapeName=dyn_skirt_back_side_l_2_capsule, center=<2 -0.1 0>, rotation=<rot 0.0001 0.0002 -22.8717>}]
 dyn_skirt_front_l_0 [{shapeType=Capsule, length=12, radius=3.5, shapeName=dyn_skirt_front_l_0_capsule, center=<2 -3.5 0>}]
 dyn_skirt_front_l_1 [{shapeType=Capsule, length=12, radius=5, shapeName=dyn_skirt_front_l_1_capsule, center=<2 -4 0>}]
 dyn_skirt_front_l_2 [{shapeType=Capsule, length=12, radius=4, shapeName=dyn_skirt_front_l_2_capsule, center=<2 -1.5 0>}]
 dyn_skirt_side_l_0 [{shapeType=Capsule, length=12, radius=3.5, shapeName=dyn_skirt_side_l_0_capsule, center=<2 -4 0>}]
 dyn_skirt_side_l_1 [{shapeType=Capsule, length=11, radius=4.5, shapeName=dyn_skirt_side_l_1_capsule, center=<2 -5.2 0>}]
 dyn_skirt_side_l_2 [{shapeType=Capsule, length=11, radius=4.5, shapeName=dyn_skirt_side_l_2_capsule, center=<2 -2.58 0>}]
 thigh_l [{shapeType=Capsule, length=32.8783, radius=9.2217, shapeName=thigh_r_capsule, center=<-17.8756 -0.3617 0>, rotation=<rot 90 175.7955 0>}]
}
constraints [
 {bone1Name=dyn_skirt_back_side_l_0, bone2Name=base_skirt_back_side_l_0, swing1LimitDegrees=25, swing2LimitDegrees=0, twistLimitDegrees=0}
 {bone1Name=dyn_skirt_back_side_l_1, bone2Name=dyn_skirt_back_side_l_0, swing1LimitDegrees=20, swing2LimitDegrees=0, twistLimitDegrees=0}
 {bone1Name=dyn_skirt_back_side_l_2, bone2Name=dyn_skirt_back_side_l_1, swing1LimitDegrees=20, swing2LimitDegrees=0, twistLimitDegrees=0}
 {bone1Name=dyn_skirt_front_l_0, bone2Name=base_skirt_front_l_0, swing1LimitDegrees=10, swing2LimitDegrees=0, twistLimitDegrees=0}
 {bone1Name=dyn_skirt_front_l_1, bone2Name=dyn_skirt_front_l_0, swing1LimitDegrees=10, swing2LimitDegrees=0, twistLimitDegrees=0}
 {bone1Name=dyn_skirt_front_l_2, bone2Name=dyn_skirt_front_l_1, swing1LimitDegrees=10, swing2LimitDegrees=0, twistLimitDegrees=0}
 {bone1Name=dyn_skirt_side_l_0, bone2Name=base_skirt_side_l_0, swing1LimitDegrees=10, swing2LimitDegrees=0, twistLimitDegrees=0}
 {bone1Name=dyn_skirt_side_l_1, bone2Name=dyn_skirt_side_l_0, swing1LimitDegrees=10, swing2LimitDegrees=0, twistLimitDegrees=0}
 {bone1Name=dyn_skirt_side_l_2, bone2Name=dyn_skirt_side_l_1, swing1LimitDegrees=10, swing2LimitDegrees=0, twistLimitDegrees=0}
]
