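# Post-process-style AnimBP for the M_MED_ParadeLotus cosmetic body, adding secondary motion to skirt and biceps.
# Parent mesh pose feeds a state-driven Control Rig, then two thigh-relative RigidBody skirt sims.
# AnimDynamics bicep springs and a front-end skirt-bone nudge follow, with bPbBlendEnabled selecting between cached poses with or without them.
# EventGraph computes gravity/RBAN settings from parameters and treats grind-rail as DBNO; a duplicate bicep chain is unconnected.
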AnimBlueprint /BRCosmetics/Characters/Player/Male/Medium/Bodies/M_MED_ParadeLotus/Meshes/M_MED_ParadeLotus_AnimBP
target_skeleton=&/M_MED_ParadeLotus_Skeleton
graphs [
 AnimGraph {nodes [
   n1 = Anim Dynamics - Bone: upperarm_bicep_l(Alpha=1, ExternalForce="0,0,0") {
    angularSpringConstant=15000
    bLinearSpring=true
    bOverrideLinearDamping=true
    bUseGravityOverride=true
    bUsePlanarLimit=false
    linearDampingOverride=1
    linearSpringConstant=500
    simulationSpace=RootRelative
    boundBone {boneName=upperarm_bicep_l}
    physicsBodyDefinitions [{
      boundBone {boneName=upperarm_bicep_l}
      boxExtents=<38 38 38>
      constraintSetup {angularLimitsMax=<0 0 0>, angularLimitsMin=<0 0 0>, angularTarget=<1 0 0>, linearAxesMax=<0 0 0.5>, linearAxesMin=<0 0 -0.5>}
      localJointOffset=<-2 0 0>
     }]
    relativeSpaceBone {boneName=upperarm_l}
    simSpaceSettings {maxAngularAcceleration=40, maxAngularVelocity=770, simSpaceAngularAlpha=1}
   }
   n2 = Anim Dynamics - Bone: upperarm_bicep_l(ComponentPose=@n1, Alpha=1, ExternalForce="0,0,0") {
    angularDampingOverride=0.9999
    angularSpringConstant=15000
    bAngularSpring=true
    bOverrideAngularDamping=true
    bUseGravityOverride=true
    bUsePlanarLimit=false
    simulationSpace=RootRelative
    boundBone {boneName=upperarm_bicep_l}
    gravityOverride {z=-980}
    physicsBodyDefinitions [{
      boundBone {boneName=upperarm_bicep_l}
      boxExtents=<24 38 24>
      constraintSetup {angularLimitsMax=<1 10 1>, angularLimitsMin=<-1 -10 -1>, angularTarget=<1 0 0>}
      localJointOffset=<-5 0 0>
     }]
    relativeSpaceBone {boneName=upperarm_l}
    simSpaceSettings {maxAngularAcceleration=10, maxAngularVelocity=300, simSpaceAngularAlpha=1}
   }
   n3 = Anim Dynamics - Bone: upperarm_bicep_r(ComponentPose=@n2, Alpha=1, ExternalForce="0,0,0") {
    angularSpringConstant=15000
    bLinearSpring=true
    bOverrideLinearDamping=true
    bUseGravityOverride=true
    bUsePlanarLimit=false
    linearDampingOverride=1
    linearSpringConstant=500
    simulationSpace=RootRelative
    boundBone {boneName=upperarm_bicep_r}
    physicsBodyDefinitions [{
      boundBone {boneName=upperarm_bicep_r}
      boxExtents=<38 38 38>
      constraintSetup {angularLimitsMax=<0 0 0>, angularLimitsMin=<0 0 0>, angularTarget=<1 0 0>, linearAxesMax=<0 0 0.5>, linearAxesMin=<0 0 -0.5>}
      localJointOffset=<2 0 0>
     }]
    relativeSpaceBone {boneName=upperarm_r}
    simSpaceSettings {maxAngularAcceleration=40, maxAngularVelocity=770, simSpaceAngularAlpha=1}
   }
   Anim Dynamics - Bone: upperarm_bicep_r(ComponentPose=@n3, Alpha=1, ExternalForce="0,0,0") {
    angularDampingOverride=0.9999
    angularSpringConstant=15000
    bAngularSpring=true
    bOverrideAngularDamping=true
    bUseGravityOverride=true
    bUsePlanarLimit=false
    simulationSpace=RootRelative
    boundBone {boneName=upperarm_bicep_r}
    gravityOverride {z=-980}
    physicsBodyDefinitions [{
      boundBone {boneName=upperarm_bicep_r}
      boxExtents=<24 38 24>
      constraintSetup {angularLimitsMax=<1 10 1>, angularLimitsMin=<-1 -10 -1>, angularTarget=<1 0 0>}
      localJointOffset=<-5 0 0>
     }]
    relativeSpaceBone {boneName=upperarm_r}
    simSpaceSettings {maxAngularAcceleration=10, maxAngularVelocity=300, simSpaceAngularAlpha=1}
   }
   Output Pose(Result=Blend Poses by bool(
     BlendPose_0=Use cached pose 'dyn_cache'()
     BlendPose_1=Use cached pose 'predyn_cache'()
     BlendTime_0=0
     BlendTime_1=0
     bActiveValue=Get bPbBlendEnabled() -> (bPbBlendEnabled:bool)
    )) {layerGroup=None}
   Save cached pose 'predyn_cache'(Pose=Use cached pose 'control_rig_cache'()) {cacheName=predyn_cache}
   n4 = Anim Dynamics - Bone: upperarm_bicep_l(ComponentPose=Local To Component(LocalPose=Use cached pose 'predyn_cache'()), Alpha=1, ExternalForce="0,0,0") {
    angularSpringConstant=15000
    bLinearSpring=true
    bOverrideLinearDamping=true
    bUseGravityOverride=true
    bUsePlanarLimit=false
    linearDampingOverride=1
    linearSpringConstant=500
    simulationSpace=RootRelative
    boundBone {boneName=upperarm_bicep_l}
    physicsBodyDefinitions [{
      boundBone {boneName=upperarm_bicep_l}
      boxExtents=<5 5 5>
      constraintSetup {angularLimitsMax=<0 0 0>, angularLimitsMin=<0 0 0>, angularTarget=<1 0 0>, linearAxesMax=<0 0 0.5>, linearAxesMin=<0 0 -0.5>}
      localJointOffset=<-2 0 0>
     }]
    relativeSpaceBone {boneName=upperarm_l}
    simSpaceSettings {maxAngularAcceleration=40, maxAngularVelocity=770, simSpaceAngularAlpha=1}
   }
   n5 = Anim Dynamics - Bone: upperarm_bicep_l(ComponentPose=@n4, Alpha=1, ExternalForce="0,0,0") {
    angularDampingOverride=0.9999
    angularSpringConstant=15000
    bAngularSpring=true
    bOverrideAngularDamping=true
    bUseGravityOverride=true
    bUsePlanarLimit=false
    simulationSpace=RootRelative
    boundBone {boneName=upperarm_bicep_l}
    gravityOverride {z=-980}
    physicsBodyDefinitions [{
      boundBone {boneName=upperarm_bicep_l}
      boxExtents=<24 38 24>
      constraintSetup {angularLimitsMax=<1 10 1>, angularLimitsMin=<-1 -10 -1>, angularTarget=<1 0 0>}
      localJointOffset=<-5 0 0>
     }]
    relativeSpaceBone {boneName=upperarm_l}
    simSpaceSettings {maxAngularAcceleration=10, maxAngularVelocity=300, simSpaceAngularAlpha=1}
   }
   n6 = Anim Dynamics - Bone: upperarm_bicep_r(ComponentPose=@n5, Alpha=1, ExternalForce="0,0,0") {
    angularSpringConstant=15000
    bLinearSpring=true
    bOverrideLinearDamping=true
    bUseGravityOverride=true
    bUsePlanarLimit=false
    linearDampingOverride=1
    linearSpringConstant=500
    simulationSpace=RootRelative
    boundBone {boneName=upperarm_bicep_r}
    physicsBodyDefinitions [{
      boundBone {boneName=upperarm_bicep_r}
      boxExtents=<5 5 5>
      constraintSetup {angularLimitsMax=<0 0 0>, angularLimitsMin=<0 0 0>, angularTarget=<1 0 0>, linearAxesMax=<0 0 0.5>, linearAxesMin=<0 0 -0.5>}
      localJointOffset=<2 0 0>
     }]
    relativeSpaceBone {boneName=upperarm_r}
    simSpaceSettings {maxAngularAcceleration=40, maxAngularVelocity=770, simSpaceAngularAlpha=1}
   }
   n7 = Anim Dynamics - Bone: upperarm_bicep_r(ComponentPose=@n6, Alpha=1, ExternalForce="0,0,0") {
    angularDampingOverride=0.9999
    angularSpringConstant=15000
    bAngularSpring=true
    bOverrideAngularDamping=true
    bUseGravityOverride=true
    bUsePlanarLimit=false
    simulationSpace=RootRelative
    boundBone {boneName=upperarm_bicep_r}
    gravityOverride {z=-980}
    physicsBodyDefinitions [{
      boundBone {boneName=upperarm_bicep_r}
      boxExtents=<24 38 24>
      constraintSetup {angularLimitsMax=<1 10 1>, angularLimitsMin=<-1 -10 -1>, angularTarget=<1 0 0>}
      localJointOffset=<-5 0 0>
     }]
    relativeSpaceBone {boneName=upperarm_r}
    simSpaceSettings {maxAngularAcceleration=10, maxAngularVelocity=300, simSpaceAngularAlpha=1}
   }
   n8 = "Transform (Modify) Bone - Bone: dyn_skirt_back_side_l_2"!(ComponentPose=@n7, bAlphaBoolEnabled:bool=Get bIsInFrontEnd(), Translation="-4,0,0", Rotation="0,0,0", Scale="1,1,1") -> (Pose) {alphaInputType=Bool, translationMode=BMM_Additive, boneToModify {boneName=dyn_skirt_back_side_l_2}, translation {x=-4}}
   Save cached pose 'dyn_cache'(Pose=Component To Local(ComponentPose=@n8)) {cacheName=dyn_cache}
   n9 = Control Rig!(
    Source=Copy Pose From Mesh() {bUseAttachedParent=true}
    Alpha
    IsCrouching:bool=Get bIsCrouching()
    IsSliding:bool=Get bIsKneeSliding()
    "Istactical ":bool=Get bIsTacticalSprinting()
    isDBNO:bool=Get IsDBNO() -> (IsDBNO:bool)
   ) -> (Pose) {controlRigAssetReference {blueprintRigClass=&/M_MED_ParadeLotus_CtrlRig2.M_MED_ParadeLotus_CtrlRig2_C}}
   n10 = RigidBody!(
    ComponentPose=Local To Component(LocalPose=@n9)
    bAlphaBoolEnabled:bool
    bUseLocalLODThresholdOnly:bool=False
    OverrideWorldGravity:Vector="0,0,0"
    ExternalForce
   ) -> (Pose) {
    alphaInputType=Bool
    bClampLinearTranslationLimitToRefPose=true
    bOverrideWorldGravity=true
    overridePhysicsAsset=&/M_MED_ParadeLotus_Physics_Skirt
    simulationSpace=BaseBoneSpace
    alphaBoolBlend {blendOption=Linear}
    alphaScaleBiasClamp {bInterpResult=true, interpSpeedDecreasing=5, interpSpeedIncreasing=5}
    baseBoneRef {boneName=thigh_l}
    simSpaceSettings {
     dampingAlpha=0
     maxAngularAcceleration=9
     maxAngularVelocity=4
     maxLinearAcceleration=1500
     maxLinearVelocity=400
     velocityScaleZ=0.75
     worldAlpha=1
     externalLinearDragV=<1.1 1.1 1.1>
    }
   }
   n11 = RigidBody(ComponentPose=@n10, bUseLocalLODThresholdOnly=False, OverrideWorldGravity="0,0,0") {
    alphaInputType=Bool
    bClampLinearTranslationLimitToRefPose=true
    bOverrideWorldGravity=true
    overridePhysicsAsset=&/M_MED_ParadeLotus_Physics_SkirtLegR
    simulationSpace=BaseBoneSpace
    alphaBoolBlend {blendOption=Linear}
    alphaScaleBiasClamp {bInterpResult=true, interpSpeedDecreasing=5, interpSpeedIncreasing=5}
    baseBoneRef {boneName=thigh_r}
    simSpaceSettings {
     dampingAlpha=0
     maxAngularAcceleration=9
     maxAngularVelocity=4
     maxLinearAcceleration=1000
     maxLinearVelocity=400
     velocityScaleZ=0.75
     worldAlpha=1
     externalLinearDragV=<1.1 1.1 1.1>
    }
   }
   Save cached pose 'control_rig_cache'(Pose=Component To Local(ComponentPose=@n11)) {cacheName=control_rig_cache}
  ]}
 EventGraph {nodes [
   Event BlueprintUpdateAnimation()
   > n1 = Sequence()
   > n2 = "Gravity Override State Function (Parameter Asset) (Nativized)"(
    CharacterPartAnimInstance=Self-Reference()
    GameTimeInSeconds=GetGameTimeInSeconds()
    Parameters=&/M_MED_ParadeLotus_Parameters
   )
   > Set GravityOverride(GravityOverride=@n2)
   > Set RBAN Settings Out(RBAN Settings Out=@n2.OutRigidBodyAnimNodeParameters)
   Set IsDBNO(execute=@n1.then_1, IsDBNO:bool=OR Boolean(A=Get bIsDBNO(), B=Get bIsOnGrindRail())) -> (Output_Get:bool)
  ]}
]
variables [
 GravityOverride {type=Vector}
 GravityOverride_cloth {type=Vector}
 IsDBNO {type=boolean}
 RBAN Settings Out {type=FortRigidBodyAnimNodeParameters}
]
