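# Physics asset of cape colliders for the male medium ParadeLotus player body cosmetic, presumably what the simulated cape collides against.
# Capsules stack along the pelvis and five spine bones to approximate the torso, with one extra capsule on the right clavicle only.
# Arm, leg and foot bodies exist but carry no shapes.
# Three constraints link the right arm and clavicle toward spine_05, all with free swing and twist.
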PhysicsAsset /BRCosmetics/Characters/Player/Male/Medium/Bodies/M_MED_ParadeLotus/Meshes/M_MED_ParadeLotus_CapeColliders
bodies {
 calf_l []
 calf_r []
 clavicle_r [{
   shapeType=Capsule
   length=9.1679
   radius=5.9681
   shapeName=clavicle_r_capsule
   center=<-12.3621 2.4938 2.1162>
   rotation=<rot -0.8479 -87.7046 -89.8532>
  }]
 foot_l []
 foot_r []
 lowerarm_l []
 lowerarm_r []
 pelvis [{shapeType=Capsule, length=16.5, radius=16.8343, center=<-10.3679 -0.4374 0>}]
 spine_01 [{shapeType=Capsule, length=7.4477, radius=14.0547, center=<-1.7965 1.1115 0>}]
 spine_02 [{shapeType=Capsule, length=7.3264, radius=13.6322, center=<-1.772 0.6285 0.0001>, rotation=<rot 0 4.29 0>}]
 spine_03 [{shapeType=Capsule, length=8.2107, radius=14.3766, center=<0.1445 0.4715 0.0001>}]
 spine_04 [{shapeType=Capsule, length=9.4404, radius=15.0613, center=<2.766 -1.0579 0.0002>}]
 spine_05 [{shapeType=Capsule, length=21.9534, radius=10.0584, center=<-0.0091 -2.5891 0>, rotation=<rot 0 7.2771 0>}]
 thigh_l []
 thigh_r []
 upperarm_l []
 upperarm_r []
}
constraints [
 {
  bone1Name=upperarm_r
  bone2Name=clavicle_r
  swing1LimitDegrees=45
  swing1Motion=Free
  swing2LimitDegrees=45
  swing2Motion=Free
  twistLimitDegrees=45
  twistMotion=Free
 }
 {
  bone1Name=lowerarm_r
  bone2Name=clavicle_r
  swing1LimitDegrees=45
  swing1Motion=Free
  swing2LimitDegrees=45
  swing2Motion=Free
  twistLimitDegrees=45
  twistMotion=Free
 }
 {
  bone1Name=clavicle_r
  bone2Name=spine_05
  swing1LimitDegrees=45
  swing1Motion=Free
  swing2LimitDegrees=45
  swing2Motion=Free
  twistLimitDegrees=45
  twistMotion=Free
 }
]
