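# Control Rig for the ParadeLotus medium male face accessory part that applies pose-driven corrections to neck helper bones as the head moves.
# A setup event calls a shared FK-control function, and the forward solve calls its setter with FK control off.
# Three chained spherical pose readers track differently oriented head regions, each weighting a ModifyTransforms offset on bones including base_helper_neck_01/02.
# Debug drawing is enabled on the readers, and headSwing_rot is unused.
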ControlRigBlueprint /BRCosmetics/Characters/Player/Male/Medium/Bodies/M_MED_ParadeLotus/Meshes/Parts/M_MED_ParadeLotus_FaceAcc_CtrlRig
graphs [
 RigVMFunctionLibrary {links [], nodes []}
 RigVMModel {nodes [
   n1 = &RigUnit_PrepareForExecution()
   addFKCtrl = RigVMFunctionReferenceNode(
    Custom_FKCtrl=()
    Options_FKCtrl="(pelvis_1_#1=false,spine_3_#2=false,clavicle_5_#3=false,arm_7_#4=false,leg_9_#5=false,head_11_46DC15EF465C ...(truncated)"
    Shape_Settings="(pelvis_2_#6=(bVisible=True,Name=\"Hexagon_Thin\",Color=(R=0.0117,G=0.8281,B=1,A=1),Transform=(Rotation=(X=0,Y=-0.7071,Z=0,W=0.7071),Translation=(X=2,Y=0,Z=0),Scale3D=(X=5,Y=5,Z=7))),spine_4_#7=(bVi ...(truncated)"
    ExecuteContext=@n1.ExecutePin
   )
   RigUnit_BeginExecution = &RigUnit_BeginExecution()
   setFKCtrl = RigVMFunctionReferenceNode(setFK_Ctrl=false, ExecuteContext=@RigUnit_BeginExecution.ExecutePin)
   RigVMFunction_Sequence = &RigVMFunction_Sequence(ExecuteContext=@setFKCtrl.ExecuteContext)
   n2 = &RigUnit_SphericalPoseReader(
    ActiveRegionScaleFactors=(PositiveWidth=1,NegativeWidth=0.75,PositiveHeight=0.75,NegativeHeight=0.75)
    ActiveRegionSize=0.1
    Debug=(bDrawDebug=true,DebugScale=25,DebugSegments=20,DebugThickness=0.25)
    DriverAxis=(X=0,Y=1,Z=0)
    DriverItem=(Type=Bone,Name="head")
    FalloffRegionScaleFactors=(PositiveWidth=1,NegativeWidth=0.76,PositiveHeight=1,NegativeHeight=1)
    FalloffSize=0.3
    OptionalParentItem=(Type=Bone,Name="neck_01")
    RotationOffset=(X=90,Y=0,Z=-45)
    ExecutePin=@RigVMFunction_Sequence.A
   )
   > &RigUnit_ModifyTransforms(
    ItemToModify="((Item=(Type=Bone,Name=\"base_helper_neck_02\"),Transform=(Rotation=(X=0,Y=0,Z=-0.6634,W=0.7483),Translation=(X=2.8445,Y=-3.4105,Z=0.8544),Scale3D=(X=1,Y=1,Z=1))),(Item=(Type=Bone,Name=\"base_helper_neck_01\"),Transform=(Rotation=(X=0,Y=0,Z=-0.0003,W=1),Transl ...(truncated)"
    WeightMinimum=0
    Weight=@n2.OutputParam
   )
   > n3 = &RigUnit_SphericalPoseReader(
    ActiveRegionScaleFactors=(PositiveWidth=0.75,NegativeWidth=0.75,PositiveHeight=1,NegativeHeight=1)
    ActiveRegionSize=0.15
    Debug=(bDrawDebug=true,bDraw2D=false,DebugScale=25,DebugSegments=20,DebugThickness=0.25)
    DriverAxis=(X=0,Y=1,Z=0)
    DriverItem=(Type=Bone,Name="head")
    FalloffRegionScaleFactors=(PositiveWidth=1,NegativeWidth=1,PositiveHeight=1,NegativeHeight=2.2)
    FalloffSize=0.23
    OptionalParentItem=(Type=Bone)
    RotationOffset=(X=150,Y=0,Z=0)
   )
   > &RigUnit_ModifyTransforms(
    ItemToModify="((Item=(Type=Bone,Name=\"base_helper_neck_02\"),Transform=(Rotation=(X=-0.342,Y=0,Z=0,W=0.9397),Translation=(X=0,Y=-1.3,Z=0),Scale3D=(X=1,Y=1,Z=1))),(Item=(Type=Bone,Name=\"base_helper_neck_01\"),Transform=(Rotation=(X=-0.3256,Y=0,Z=0,W=0.9455),Translation=(X= ...(truncated)"
    WeightMinimum=0
    Weight=@n3.OutputParam
   )
   > n4 = &RigUnit_SphericalPoseReader(
    ActiveRegionScaleFactors=(PositiveWidth=0.75,NegativeWidth=0.75,PositiveHeight=1,NegativeHeight=1)
    ActiveRegionSize=0.15
    Debug=(bDrawDebug=true,bDraw2D=false,DebugScale=25,DebugSegments=20,DebugThickness=0.25)
    DriverAxis=(X=0,Y=1,Z=0)
    DriverItem=(Type=Bone,Name="head")
    FalloffRegionScaleFactors=(PositiveWidth=1,NegativeWidth=1,PositiveHeight=1,NegativeHeight=2.2)
    FalloffSize=0.23
    FlipHeightScaling=true
    FlipWidthScaling=false
    OptionalParentItem=(Type=Bone)
    RotationOffset=(X=30,Y=0,Z=0)
   )
   > &RigUnit_ModifyTransforms(
    ItemToModify="((Item=(Type=Bone,Name=\"base_helper_neck_02\"),Transform=(Rotation=(X=0.342,Y=0,Z=0,W=0.9397),Translation=(X=0,Y=-1.3,Z=0),Scale3D=(X=1,Y=1,Z=1))),(Item=(Type=Bone,Name=\"base_helper_neck_01\"),Transform=(Rotation=(X=0.3322,Y=0,Z=0,W=0.9432),Translation=(X=0. ...(truncated)"
    WeightMinimum=0
    Weight=@n4.OutputParam
   )
  ]}
]
hierarchy [
 root {loc=<0 0 0>, rot=<0 0 0 1>}
 pelvis {parent=root, loc=<0 -0.7071 93.1191>, rot=<0.0127 0.707 -0.0127 -0.707>}
 spine_01 {parent=pelvis, loc=<6.7494 -0.6388 0>, rot=<0 0 0.034 -0.9994>}
 spine_02 {parent=spine_01, loc=<7.2518 -0.035 0>, rot=<0 0 0.0238 -0.9997>}
 spine_03 {parent=spine_02, loc=<6.0104 0 0>, rot=<0 0 0.0149 -0.9999>}
 spine_04 {parent=spine_03, loc=<8.1024 0 0>, rot=<0 0 -0.0132 -0.9999>}
 spine_05 {parent=spine_04, loc=<10.0146 0 0>, rot=<0 0 -0.0015 -1>}
 neck_01 {parent=spine_05, loc=<11.3138 0.1529 0>, rot=<0 0 -0.2003 -0.9797>}
 neck_02 {parent=neck_01, loc=<4.5627 0 0>, rot=<0 0 0.0504 -0.9987>}
 head {parent=neck_02, loc=<5.3914 0 0>, rot=<0 0 0.115 -0.9934>}
 base_hair_front_01 {parent=head, loc=<17.9084 9.6839 3.1767>, rot=<0.5396 0.3412 0.7433 0.2>}
 dyn_hair_front_01 {parent=base_hair_front_01, loc=<0 0 0>, rot=<0 0 0 1>}
 dyn_hair_front_02 {parent=dyn_hair_front_01, loc=<5.3326 0.0158 -0.1748>, rot=<0 0.1689 0 -0.9856>}
 dyn_hair_front_03 {parent=dyn_hair_front_02, loc=<3.1683 0.0864 0.0448>, rot=<0 -0.9122 0 -0.4097>}
 base_hair_front_01_l {parent=head, loc=<17.9084 9.6839 -1.6865>, rot=<0.3811 0.5122 0.6219 0.4536>}
 dyn_hair_front_01_l {parent=base_hair_front_01_l, loc=<0 0 0>, rot=<0 0 0 1>}
 base_helper_neck_02 {parent=neck_02, loc=<0 0 0>, rot=<0 0 0 1>}
 dyn_helper_neck_02 {parent=base_helper_neck_02, loc=<0 0 0>, rot=<0 0 0 1>}
 base_helper_neck_01 {parent=neck_01, loc=<0 0 0>, rot=<0 0 0 1>}
 dyn_helper_neck_01 {parent=base_helper_neck_01, loc=<0 0 0>, rot=<0 0 0 1>}
 clavicle_l {parent=spine_05, loc=<4.3575 -0.1466 -2.3223>, rot=<0.0465 -0.7289 -0.009 -0.6829>}
 upperarm_l {parent=clavicle_l, loc=<15.5111 0 0>, rot=<0.0226 -0.3234 -0.0088 -0.9459>}
 clavicle_r {parent=spine_05, loc=<4.3575 -0.1466 2.3223>, rot=<0.7289 0.0465 -0.6829 0.009>}
 upperarm_r {parent=clavicle_r, loc=<-15.5111 0 0>, rot=<0.0226 -0.3234 -0.0088 -0.9459>}
]
variables [headSwing_rot {type=FVector, default=(X=0,Y=0,Z=0)}]
