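# Post-process style AnimBlueprint for the ParadeLotus male face accessory part, adding secondary motion to its front hair.
# It copies the attached parent's pose, runs a Control Rig, then chains four spring-driven, root-relative Anim Dynamics nodes on front hair bones.
# The EventGraph computes gravity/RBAN settings, a skydive/parachute hair alpha and a zipline vector.
# The alpha drives neck-helper bone offsets that are disconnected from the output pose.
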AnimBlueprint /BRCosmetics/Characters/Player/Male/Medium/Bodies/M_MED_ParadeLotus/Meshes/Parts/M_MED_ParadeLotus_FaceAcc_AnimBP
target_skeleton=&/M_MED_ParadeLotus_FaceAcc_Skeleton
graphs [
 AnimGraph {nodes [
   n1 = Get skydiving_hair() -> (skydiving_hair:double)
   n2 = "Transform (Modify) Bone - Bone: dyn_helper_neck_01"(Alpha=@n1, Translation="0,-10,7", Rotation="0,0,7", Scale="1,1,1") {
    rotationMode=BMM_Additive
    translationMode=BMM_Additive
    boneToModify {boneName=dyn_helper_neck_01}
    rotation {roll=7}
    translation {y=-10, z=7}
   }
   "Transform (Modify) Bone - Bone: dyn_helper_neck_02"(ComponentPose=@n2, Alpha=@n1, Translation="0,4.7,4.5", Rotation="0,0,-35", Scale="1,1,1") {
    rotationMode=BMM_Additive
    translationMode=BMM_Additive
    boneToModify {boneName=dyn_helper_neck_02}
    rotation {roll=-35}
    translation {y=4.7, z=4.5}
   }
   n3 = Control Rig(Source=Copy Pose From Mesh() {bUseAttachedParent=true}) {controlRigAssetReference {blueprintRigClass=&/M_MED_ParadeLotus_FaceAcc_CtrlRig.M_MED_ParadeLotus_FaceAcc_CtrlRig_C}}
   n4 = Anim Dynamics - Bone: dyn_hair_front_01!(ComponentPose=Local To Component(LocalPose=@n3), Alpha=1, ComponentLinearAccScale:Vector="1,1,1", ComponentLinearVelScale:Vector="1,1,1") -> (Pose) {
    angularDampingOverride=1
    angularSpringConstant=4000
    bAngularSpring=true
    bGravityOverrideInSimSpace=true
    bOverrideAngularDamping=true
    bUseGravityOverride=true
    gravityScale=0.35
    simulationSpace=RootRelative
    alphaBoolBlend {blendOption=Linear}
    boundBone {boneName=dyn_hair_front_01}
    componentAppliedLinearAccClamp=<3000 3000 3000>
    componentLinearAccScale=<1 1 1>
    componentLinearVelScale=<1 1 1>
    gravityOverride {z=-980}
    physicsBodyDefinitions [{
      boundBone {boneName=dyn_hair_front_01}
      boxExtents=<15 15 15>
      constraintSetup {angularLimitsMax=<0 0 20>, angularLimitsMin=<0 -30 -15>, angularTarget=<1 0 0>}
      localJointOffset=<4 0 0>
     }]
    simSpaceSettings {maxAngularAcceleration=35, maxAngularVelocity=7, simSpaceAngularAlpha=0.8}
   }
   n5 = Anim Dynamics - Bone: dyn_hair_front_02!(ComponentPose=@n4, Alpha=1, ComponentLinearAccScale:Vector="1,1,1", ComponentLinearVelScale:Vector="1,1,1") -> (Pose) {
    angularDampingOverride=1
    angularSpringConstant=4000
    bAngularSpring=true
    bGravityOverrideInSimSpace=true
    bOverrideAngularDamping=true
    bUseGravityOverride=true
    gravityScale=0.35
    simulationSpace=RootRelative
    alphaBoolBlend {blendOption=Linear}
    boundBone {boneName=dyn_hair_front_02}
    componentAppliedLinearAccClamp=<3000 3000 3000>
    componentLinearAccScale=<1 1 1>
    componentLinearVelScale=<1 1 1>
    gravityOverride {z=-300}
    physicsBodyDefinitions [{
      boundBone {boneName=dyn_hair_front_02}
      boxExtents=<15 15 15>
      constraintSetup {angularLimitsMax=<0 10 20>, angularLimitsMin=<0 -30 -15>, angularTarget=<1 0 0>}
      localJointOffset=<3 0 0>
     }]
    simSpaceSettings {maxAngularAcceleration=35, maxAngularVelocity=7, simSpaceAngularAlpha=0.8}
   }
   n6 = Anim Dynamics - Bone: dyn_hair_front_03!(ComponentPose=@n5, Alpha=1, ComponentLinearAccScale:Vector="1,1,1", ComponentLinearVelScale:Vector="1,1,1") -> (Pose) {
    angularDampingOverride=1
    angularSpringConstant=20000
    bAngularSpring=true
    bGravityOverrideInSimSpace=true
    bOverrideAngularDamping=true
    bUseGravityOverride=true
    gravityScale=0.35
    simulationSpace=RootRelative
    alphaBoolBlend {blendOption=Linear}
    boundBone {boneName=dyn_hair_front_03}
    componentAppliedLinearAccClamp=<3000 3000 3000>
    componentLinearAccScale=<1 1 1>
    componentLinearVelScale=<1 1 1>
    gravityOverride {z=-980}
    physicsBodyDefinitions [{
      boundBone {boneName=dyn_hair_front_03}
      boxExtents=<30 30 30>
      constraintSetup {angularLimitsMax=<0 10 20>, angularLimitsMin=<0 -30 -15>, angularTarget=<1 0 0>}
      localJointOffset=<3 0 0>
     }]
    simSpaceSettings {maxAngularAcceleration=35, maxAngularVelocity=7, simSpaceAngularAlpha=0.8}
   }
   n7 = Anim Dynamics - Bone: dyn_hair_front_01_l!(ComponentPose=@n6, Alpha=1, ComponentLinearAccScale:Vector="1,1,1", ComponentLinearVelScale:Vector="1,1,1") -> (Pose) {
    angularDampingOverride=1
    angularSpringConstant=20000
    bAngularSpring=true
    bGravityOverrideInSimSpace=true
    bOverrideAngularDamping=true
    bUseGravityOverride=true
    gravityScale=0.35
    simulationSpace=RootRelative
    alphaBoolBlend {blendOption=Linear}
    boundBone {boneName=dyn_hair_front_01_l}
    componentAppliedLinearAccClamp=<3000 3000 3000>
    componentLinearAccScale=<1 1 1>
    componentLinearVelScale=<1 1 1>
    gravityOverride {z=-980}
    physicsBodyDefinitions [{
      boundBone {boneName=dyn_hair_front_01_l}
      boxExtents=<30 30 30>
      constraintSetup {angularLimitsMax=<0 2 20>, angularLimitsMin=<0 -15 -15>, angularTarget=<1 0 0>}
      localJointOffset=<3 0 0>
     }]
    simSpaceSettings {maxAngularAcceleration=35, maxAngularVelocity=7, simSpaceAngularAlpha=0.8}
   }
   Output Pose(Result=Component To Local(ComponentPose=@n7)) {layerGroup=None}
  ]}
 EventGraph {nodes [
   Event BlueprintUpdateAnimation()
   > n1 = Sequence()
   > n2 = "Gravity Override State Function (Parameter Asset) (Nativized)"(
    CharacterPartAnimInstance=Self-Reference()
    GameTimeInSeconds=GetGameTimeInSeconds()
    Parameters=/BRCosmetics/Characters/Player/Male/Medium/Bodies/M_MED_ParadeLotus/Meshes/M_MED_ParadeLotus_Parameters
   )
   > Set GravityOverride(GravityOverride=@n2)
   > Set RBAN Settings Out(RBAN Settings Out=@n2.OutRigidBodyAnimNodeParameters)
   n3 = NOT Boolean(A=Get bIsSkyDivingFromBus() -> (bIsSkyDivingFromBus:bool))
   n4 = NOT Boolean(A=Get bIsParachuteOpened())
   Set skydiving_hair(execute=@n1.then_1, skydiving_hair:double="To Float (Boolean)"(InBool=OR Boolean(A=@n3, B=@n4))) -> (Output_Get:double)
   n5 = Event BlueprintProcessAnimInputEvent()
   n6 = IsNamed(self=@n5.AnimInputEvent, InEventName=aie.Player.Zipline.Entered)
   n7 = OR Boolean(A=Get bIsMoving2D(), B=Get bIsTacticalSprinting())
   n8 = Get bIsJumping()
   > n9 = Branch(Condition=OR Boolean(A=@n6, B=AND Boolean!(A=@n6, B=@n8, C:bool=@n7) -> (ReturnValue)))
   > Set animDyn on zipline(animDyn on zipline:Vector="0, 0, 0") -> (Output_Get:Vector)
   Branch(execute=@n9.else, Condition=IsNamed(self=@n5.AnimInputEvent, InEventName=aie.Player.Zipline.Exited))
   > Set animDyn on zipline(animDyn on zipline="0.5,0.5,0.5")
   Get bIsSkydiveFloating()
   Get bIsSkydiveDiveMode()
   Get bIsSkyDiving()
  ]}
]
variables [
 animDyn on zipline {type=Vector}
 GravityOverride {type=Vector}
 RBAN Settings Out {type=FortRigidBodyAnimNodeParameters}
 Skydiving Hair {type=number}
 skydiving_hair {type=number}
]
